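# Additive, unlit, two-sided material for the butterfly trail in the KelpLinen skydiving contrail, built for Niagara particles including meshes.
# A tiled smoke texture, panned by a Niagara timer and offset, is refined, colored by a curve atlas and tinted by particle color.
# Depth fade, a view-angle falloff and UV edge fades shape opacity.
# Separate branches compute Niagara-weighted displacement values and a sine-shaped, mirrored offset along the UV.
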Material /BRCosmetics/Athena/Effects/Particles/Skydiving/NS_Contrail_KelpLinen/M_Contrail_KelpLinen_Butterfly_Trail
graph {
 properties {
  bUsedWithNiagaraMeshParticles=true
  blendMode=BLEND_Additive
  pixelDepthOffsetMode=PDOM_Legacy
  shadingModel=MSM_Unlit
  twoSided=true
  usedShadingModels=Unlit
 }
 nodes [
  n1 = DynamicParameter() {desc=Parameters set by Niagara, paramNames [X,Y,Z,WPOIntensity]}
  n2 = TextureCoordinate()
  n3 = OneMinus(None=ComponentMask(None=MaterialFunctionCall(XY=@n2) {materialFunction=&Swizzle}.YX) {r=true})
  n4 = ScalarParameter() {defaultValue=1, parameterName=BP}
  LinearInterpolate(
   B=MaterialFunctionCall(X=@n1.X, Y=@n1.X, Z=@n1.X) {materialFunction=&MakeFloat3}.Result
   Alpha=Multiply(A=Power(Base=@n3, Exp=@n4), B=@n1.WPOIntensity)
  )
  n5 = ParticleColor()
  n6 = MaterialFunctionCall(XY=TextureCoordinate()) {materialFunction=&Swizzle}
  n7 = CurveAtlasRowParameter(CurveTime=MaterialFunctionCall(Float2=@n6.YX) {materialFunction=&BreakOutFloat2Components}.R) {
   atlas=&AthenaCurveAtlas01
   curve=/Game/Characters/Player/Male/Medium/Bodies/M_MED_Viper/Textures/Glitch/viperGllitchColor
   defaultValue=136
  }
  n8 = VectorParameter() {parameterName=Tile2, defaultValue=<rgba 0 0 0 1>}
  n9 = Add(A=@n6.YX, B=AppendVector(A=@n8.R, B=@n8.R))
  n10 = VectorParameter() {parameterName=Tile, defaultValue=<rgba 0.25 1 0.25 0>}
  n11 = AppendVector(A=@n10.R, B=@n10.R)
  n12 = AppendVector(A=@n10.B, B=@n10.B)
  n13 = DynamicParameter() {parameterIndex=3, paramNames [Timer,Offset,Param3,Param4]}
  n14 = Add(A=Add(A=Multiply(A=@n9, B=@n11), B=Multiply(A=@n12, B=@n13.Timer)), B=AppendVector(A=@n13.Offset, B=Constant()))
  n15 = Multiply(
   A=MaterialFunctionCall!(TexCoord=@n14, ClampChannel=AppendVector(A=@n8.B, B=@n8.B)) {materialFunction=/Game/Effects/Fort_Effects/MaterialFunctions/MF_FX_TilingMethod}.Result
   B=TextureSampleParameter2D(UVs=@n14) {parameterName=Texture, samplerType=SAMPLERTYPE_Grayscale, texture=&T_Smoke_Trail_Soft}.RGBA
  )
  n16 = VectorParameter() {group=Trail, parameterName=Trail Refine, defaultValue=<rgba 1 1 1 0.5>}
  n17 = MaterialFunctionCall(Texture=ComponentMask(None=@n15) {r=true}, RGB=@n16.RGB, A=@n16.RGB) {materialFunction=&MF_FX_Refine}
  MaterialFunctionCall(Emissive=Multiply(A=@n5.RGB, B=Multiply(A=@n7, B=@n17.Result))) {materialFunction=&MF_FX_EmissiveAdjustments}
  n18 = VectorParameter() {group=Trail, parameterName=Trail ARefine, defaultValue=<rgba 1 1 1 0.5>}
  n19 = Saturate(None=MaterialFunctionCall(Texture=ComponentMask(None=@n15) {a=true}, RGB=@n18.RGB, A=@n18.RGB) {materialFunction=&MF_FX_Refine}.Result)
  n20 = VectorParameter() {parameterName=Fresnel, defaultValue=<rgba 30 0.8 0 1>}
  n21 = Add(A=DotProduct(A=PixelNormalWS(), B=CameraVectorWS()), B=@n20.G)
  n22 = Add(A=Power(Base=Clamp(None=@n21) {clampMode=CMODE_ClampMin}, Exp=@n20.R), B=@n20.B)
  n23 = DepthFade(Opacity=Multiply(A=Multiply(A=@n7.A, B=@n19), B=Saturate(None=@n22)), FadeDistance=ScalarParameter() {defaultValue=25, parameterName=FD})
  n24 = ScalarParameter() {defaultValue=1, parameterName=ADC}
  n25 = TextureCoordinate()
  n26 = ComponentMask(None=@n25) {g=true}
  n27 = Power(Base=OneMinus(None=ComponentMask(None=@n25) {g=true})) {constExponent=1}
  Multiply(A=Multiply(A=@n5.A, B=Clamp(None=@n23, Max=@n24)), B=Multiply(A=Power(Base=@n26), B=Multiply(A=@n27) {constB=2.5}))
  n28 = MaterialFunctionCall(Float2=TextureCoordinate()) {materialFunction=&BreakOutFloat2Components}
  n29 = LinearInterpolate(
   A=Constant3Vector() {constant=<rgba 0 -1 0 0>}
   B=Constant3Vector() {constant=<rgba 0 1 0 0>}
   Alpha=Floor(None=Multiply(A=@n28.G) {constB=2})
  )
  n30 = Saturate(None=@n28.G)
  n31 = Subtract(A=MaterialFunctionCall(Value=Add(A=@n30) {constB=0.25}) {materialFunction=&LinearSine}.Linear Sine) {constB=0.01}
  Multiply(A=Multiply(A=Transform(None=@n29), B=Saturate(None=@n31)), B=ScalarParameter() {defaultValue=50, parameterName=Param_1})
 ]
}
