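# Translucent unlit Niagara material for the KelpLinen skydiving contrail, usable on ribbons, sprites and mesh particles.
# Flowmap, noise and panned shape textures build an eroded trail with dark edges, iridescent curve-atlas color, sphere masks and time-of-day tints.
# Niagara dynamic parameters supply timer, offsets, hand flip, audio scale and bend inputs.
# A PolarCoordinate static switch remaps UVs and adds Fresnel; layered sines drive world position offset, optionally adding an interpolated bend.
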Material /BRCosmetics/Athena/Effects/Particles/Skydiving/NS_Contrail_KelpLinen/M_KelpLinen_Trail
graph {
 properties {
  bAutomaticallySetUsageInEditor=false
  bUsedWithNiagaraMeshParticles=true
  bUsedWithNiagaraRibbons=true
  bUsedWithNiagaraSprites=true
  blendMode=BLEND_Translucent
  shadingModel=MSM_Unlit
  twoSided=true
  usedShadingModels=Unlit
 }
 nodes [
  n1 = Multiply(
   A=TextureSample(UVs=NamedRerouteUsage()) {samplerType=SAMPLERTYPE_LinearColor, texture=/Game/VisualThreatManager/StormVisuals/LargeClouds/T_Swirl_Flowmap}.RGB
   B=ScalarParameter() {defaultValue=0.8, parameterName=Distortion Strength}
  )
  NamedRerouteDeclaration(None=Add(A=ComponentMask(None=@n1) {g=true, r=true}, B=Constant2Vector() {g=-0.375})) {name=Distortion, nodeColor=<rgba 0 0.3168 1 1>}
  n2 = TextureCoordinate()
  n3 = ComponentMask(None=@n2) {g=true}
  n4 = OneMinus(None=@n3)
  n5 = DynamicParameter() {defaultValue=<rgba 1 0 0 0.5>, paramNames [Timer, Offset, InverseHand, Sprite_Offset]}
  n6 = Add(
   A=AppendVector(A=ComponentMask(None=@n2) {r=true}, B=LinearInterpolate(A=@n3, B=@n4, Alpha=@n5.InverseHand))
   B=Multiply(A=@n5.Offset, B=Constant2Vector() {r=1})
  )
  n7 = MaterialFunctionCall!(UV Coord=Add(A=@n6, B=Constant2Vector() {g=-0.5}), LenghtScale, RadialScale) {materialFunction=/Game/Global/MaterialFunctions/TechArt/MF_PolarCoordinates}
  n8 = AppendVector(A=@n5.Sprite_Offset, B=Constant())
  NamedRerouteDeclaration(None=StaticSwitchParameter(True=Add(A=@n7.Result, B=@n8), False=@n6) {parameterName=PolarCoordinate}) {name=TexCoord, nodeColor=<rgba 1 0.2344 0 1>}
  n9 = DynamicParameter() {desc=Parameters set by Niagara, parameterIndex=1, paramNames [X,Y,Z,WPOIntensity]}
  n10 = Clamp(
   None=Subtract(
    A=MaterialFunctionCall(X=@n9.X, Y=@n9.X, Z=@n9.X) {materialFunction=&MakeFloat3}.Result
    B=ParticlePositionWS()
   )
   Min=ScalarParameter() {defaultValue=-100, group=Interpolate WPO, parameterName=Clamp_min, sortPriority=1}
   Max=ScalarParameter() {defaultValue=100, group=Interpolate WPO, parameterName=Clamp_max, sortPriority=2}
  ) {maxDefault=1000, minDefault=-1000}
  n11 = OneMinus(None=TextureCoordinate())
  n12 = Multiply(
   A=@n9.WPOIntensity
   B=Power(
    Base=ComponentMask(None=@n11) {g=true}
    Exp=ScalarParameter() {defaultValue=3, group=Interpolate WPO, parameterName=WPOBendPower, sortPriority=3}
   )
  )
  NamedRerouteDeclaration(None=LinearInterpolate(
    A=Constant()
    B=LinearInterpolate(B=@n10, Alpha=@n12)
    Alpha=ScalarParameter() {defaultValue=1, group=Interpolate WPO, parameterName=Use interpolate WPO, sortPriority=0}
   )) {name=Interpolate WPO, nodeColor=<rgba 0 0.4609 1 1>}
  n13 = TextureObjectParameter() {
   group=BaseShape
   parameterName=BaseShape_Texture
   samplerType=SAMPLERTYPE_LinearColor
   sortPriority=0
   texture=&/T_KelpLinen_Trail
  }
  n14 = VectorParameter() {group=BaseShape, parameterName=BaseShape_UVControl, channelNames {a=VPan, b=UPan, g=VScale, r=UScale}, defaultValue=<rgba 0.75 1 0 0>}
  n15 = NamedRerouteUsage()
  n16 = Add(A=Add(A=NamedRerouteUsage(), B=NamedRerouteUsage()), B=NamedRerouteUsage())
  n17 = MaterialFunctionCall(Texture=@n13, UVControl_RGB=@n14, UVControl_A=@n14, "[Override]TimeInput"=@n15, "[Override]UVInput"=@n16) {materialFunction=&FXCommon_TextureSample}
  n18 = TextureObjectParameter() {
   group=BaseShape
   parameterName=Grayscale_Texture
   samplerType=SAMPLERTYPE_LinearGrayscale
   sortPriority=0
   texture=&/T_KelpLinen_Line_Grayscale
  }
  n19 = MaterialFunctionCall(
   R=@n17.R
   G=@n17.R
   B=MaterialFunctionCall(Texture=@n18, UVControl_RGB=@n14, UVControl_A=@n14, "[Override]TimeInput"=@n15, "[Override]UVInput"=@n16) {materialFunction=&FXCommon_TextureSample}.G
   A=@n17.R
  ) {materialFunction=&AppendMany}
  NamedRerouteDeclaration(None=@n19.RG) {name=UVShape, nodeColor=<rgba 1 0.5156 0 1>}
  n20 = CurveAtlasRowParameter(CurveTime=Frac(None=NamedRerouteUsage())) {
   atlas=&AthenaCurveAtlas01
   curve=/Game/Athena/LinearColorCurves/CC_KelpLinen_Trail_Iridescence
   defaultValue=343
  }
  n21 = Constant() {r=0.75}
  n22 = TextureCoordinate()
  n23 = MaterialFunctionCall(UV Coord=Add(A=@n22, B=Constant2Vector() {g=-0.5})) {materialFunction=&MF_PolarCoordinates}
  n24 = OneMinus(None=ComponentMask(None=StaticSwitchParameter(True=@n23.Result, False=@n22) {parameterName=PolarCoordinate}) {r=true})
  n25 = Multiply(A=Saturate(None=Desaturation(None=@n20, Fraction=@n21)), B=@n24)
  n26 = Multiply(A=@n24, B=NamedRerouteUsage())
  n27 = NamedRerouteUsage()
  n28 = Saturate(None=Multiply(A=Power(Base=@n26), B=ComponentMask(None=@n27) {b=true}))
  NamedRerouteDeclaration(None=Saturate(None=Add(A=@n25, B=@n28))) {name=Iridescent, nodeColor=<rgba 0.1953 0 1 1>}
  n29 = TextureObjectParameter() {
   group=NoiseA
   parameterName=NoiseA_Texture
   samplerType=SAMPLERTYPE_LinearGrayscale
   sortPriority=0
   texture=&T_SimplePerlin_Greyscale
  }
  n30 = VectorParameter() {group=NoiseA, parameterName=NoiseA_UVControl, channelNames {a=VPan, b=UPan, g=VScale, r=UScale}, defaultValue=<rgba 0.3 0.3 0.25 0>}
  n31 = NamedRerouteUsage()
  n32 = NamedRerouteUsage()
  n33 = VectorParameter() {
   group=NoiseA
   parameterName=NoiseA_NoiseControl
   sortPriority=60
   channelNames {a=VStrength, b=UStrength, g=VOffset, r=UOffset}
   defaultValue=<rgba -0.5 -0.5 0 -0.25>
  }
  NamedRerouteDeclaration(None=MaterialFunctionCall!(
    NoiseValue=MaterialFunctionCall(Texture=@n29, UVControl_RGB=@n30, UVControl_A=@n30, "[Override]TimeInput"=@n31, "[Override]UVInput"=@n32) {materialFunction=&FXCommon_TextureSample}.RG
    Control_RGB=@n33
    Control_A=@n33
   ) {materialFunction=/Game/Effects/Niagara/MasterMaterials/Functions/Common/FXCommon_NoiseControl}.NoiseValue) {name=Noise, nodeColor=<rgba 0.125 0 1 1>}
  n34 = TextureCoordinate()
  n35 = Constant2Vector() {g=0.5}
  n36 = StaticSwitchParameter(
   True=MaterialFunctionCall(UV Coord=Add(A=@n34, B=@n35)) {materialFunction=&MF_PolarCoordinates}.Result
   False=@n34
  ) {parameterName=PolarCoordinate}
  n37 = ComponentMask(None=@n36) {g=true}
  n38 = OneMinus(None=ComponentMask(None=@n36) {g=true})
  NamedRerouteDeclaration(None=Saturate(None=Multiply(A=@n37, B=@n38))) {name=GradientV_Mirror, nodeColor=<rgba 1 0.4453 0 1>}
  n39 = NamedRerouteUsage()
  n40 = Frac(None=@n39)
  n41 = Multiply(A=ComponentMask(None=NamedRerouteUsage()) {b=true}, B=LinearInterpolate(Alpha=Saturate(None=@n40)) {constA=0.5, constB=0.6}) {constB=0.5}
  n42 = Multiply(A=Add(A=@n39, B=NamedRerouteUsage()) {constB=0}) {constB=0.25}
  n43 = ComponentMask(None=NamedRerouteUsage()) {g=true}
  n44 = OneMinus(None=ComponentMask(None=TextureCoordinate()) {r=true})
  n45 = Frac(None=Add(A=Add(A=@n41, B=@n42), B=Add(A=@n43, B=@n44)))
  NamedRerouteDeclaration(None=ConstantBiasScale(None=Saturate(None=@n45)) {bias=0, scale=1}) {name=HueShift, nodeColor=<rgba 0.125 1 0 1>}
  n46 = TextureCoordinate()
  n47 = MaterialFunctionCall(UV Coord=Add(A=@n46, B=Constant2Vector() {g=-0.5})) {materialFunction=&MF_PolarCoordinates}
  n48 = VectorParameter() {parameterName=ScaleCC, defaultValue=<rgba 1 2.5 0 1>}
  n49 = TextureObjectParameter() {
   group=Dissolve
   parameterName=Dissolve_Texture
   samplerType=SAMPLERTYPE_LinearGrayscale
   sortPriority=0
   texture=/Game/Global/Textures/Default/Noise/T_Bubble_B
  }
  n50 = VectorParameter() {group=Dissolve, parameterName=Dissolve_UVControl, channelNames {a=VPan, b=UPan, g=VScale, r=UScale}, defaultValue=<rgba 0.25 0 -3 0>}
  n51 = Add(A=MaterialFunctionCall(Texture=@n49, UVControl_RGB=@n50, UVControl_A=@n50, "[Override]TimeInput"=NamedRerouteUsage(), "[Override]UVInput"=NamedRerouteUsage()) {materialFunction=&FXCommon_TextureSample}.R) {constB=-0.3333}
  n52 = OneMinus(None=Power(Base=OneMinus(None=@n51), Exp=ScalarParameter() {defaultValue=10, parameterName=ScalePower}) {constExponent=5})
  n53 = MaterialFunctionCall(
   UVs=Add(A=StaticSwitchParameter(True=@n47.Result, False=@n46) {parameterName=PolarCoordinate}, B=Constant2Vector() {g=0.025})
   Texture Scale=MaterialFunctionCall(R=@n48.R, G=LinearInterpolate(B=@n48.G, Alpha=@n52)) {materialFunction=&AppendMany}.RG
  ) {materialFunction=&ScaleUVsByCenter}
  n54 = NamedRerouteUsage()
  n55 = Add(A=@n53.UVs, B=@n54) {constB=0}
  n56 = Constant() {r=0.5}
  n57 = SphereMask(A=@n55, B=@n56) {attenuationRadius=0.49, hardnessPercent=50}
  n58 = Saturate(None=Add(A=NamedRerouteUsage()) {constB=-0.025})
  n59 = Multiply(A=@n58) {constB=10}
  NamedRerouteDeclaration(None=Saturate(None=Multiply(A=@n57, B=@n59))) {name=SphereMaskLine, nodeColor=<rgba 0.9688 0 1 1>}
  n60 = SphereMask(A=@n55, B=@n56) {attenuationRadius=0.49}
  n61 = Multiply(A=@n58) {constB=100}
  NamedRerouteDeclaration(None=Saturate(None=Multiply(A=@n60, B=@n61))) {name=SphereMask, nodeColor=<rgba 0.6406 1 0 1>}
  NamedRerouteDeclaration(None=DynamicParameter() {parameterIndex=2, defaultValue=<rgba 0 1 1 1>, paramNames [AudioScale,Param2,Param3,Param4]}.AudioScale) {name=AudioScale, nodeColor=<rgba 0.9922 0 1 1>}
  NamedRerouteDeclaration(None=Constant() {r=0.72}) {name=ParticleAlpha, nodeColor=<rgba 1 0.3516 0 1>}
  n62 = VectorParameter() {parameterName=TotalUVControl, defaultValue=<rgba 1 1 -0.2 0>}
  NamedRerouteDeclaration(None=MaterialFunctionCall!(UVInput=NamedRerouteUsage(), TimeInput=NamedRerouteUsage(), Control_RGB=@n62.RGB, Control_A=@n62.RGB) {materialFunction=/Game/Effects/Niagara/MasterMaterials/Functions/Common/FXCommon_QuickUVControl}.UV) {name=UV, nodeColor=<rgba 0 1 0.3672 1>}
  NamedRerouteDeclaration(None=MaterialFunctionCall(
    Color_Morning=VectorParameter() {parameterName=Color Morning, defaultValue=<rgba 1 0.5 0.5033 1>}.RGB
    Color_Day=VectorParameter() {parameterName=Color Day, defaultValue=<rgba 1 1 1 1>}.RGB
    Color_Evening=VectorParameter() {parameterName=Color Evening, defaultValue=<rgba 0.9 0.36 0.3649 1>}.RGB
    Color_Night=VectorParameter() {parameterName=Color Night, defaultValue=<rgba 0.3259 0.412 0.9 1>}.RGB
   ) {materialFunction=&Color_Time_of_Day}.Result) {name=TOD, nodeColor=<rgba 1 0.6563 0 1>}
  n63 = Multiply(A=@n54) {constB=0.2}
  n64 = MaterialFunctionCall(UVs=Add(A=Add(A=@n63, B=@n53.UVs), B=Constant2Vector() {r=0.5}), Texture Scale=Constant2Vector() {g=0.35, r=1.5}) {materialFunction=&ScaleUVsByCenter}
  NamedRerouteDeclaration(None=Saturate(None=SphereMask(A=@n64.UVs, B=@n56) {attenuationRadius=0.49, hardnessPercent=0})) {name=SoftSphereMask, nodeColor=<rgba 0 0.0625 1 1>}
  n65 = NamedRerouteUsage()
  n66 = Saturate(None=Multiply(A=ComponentMask(None=@n65) {r=true}))
  n67 = ComponentMask(None=@n65) {g=true}
  n68 = ComponentMask(None=NamedRerouteUsage()) {b=true}
  n69 = Divide(A=Subtract(A=Multiply(A=@n66, B=@n67), B=Power(Base=@n68) {constExponent=2.5})) {constB=0.5}
  NamedRerouteDeclaration(None=Divide(A=Saturate(None=@n69)) {constB=0.25}) {name=OutlineWhite, nodeColor=<rgba 0.7578 1 0 1>}
  n70 = NamedRerouteUsage()
  n71 = ComponentMask(None=@n70) {r=true}
  n72 = Multiply(A=Saturate(None=Multiply(A=@n71) {constB=5}), B=ComponentMask(None=@n70) {g=true})
  n73 = NamedRerouteUsage()
  n74 = TextureCoordinate()
  n75 = MaterialFunctionCall(UV Coord=Add(A=@n74, B=Constant2Vector() {g=-0.5})) {materialFunction=&MF_PolarCoordinates}
  n76 = ComponentMask(None=NamedRerouteUsage()) {g=true}
  n77 = Saturate(None=Add(A=StaticSwitchParameter(True=@n75.Result, False=@n74) {parameterName=PolarCoordinate}, B=Multiply(A=@n76) {constB=0.75}))
  n78 = ComponentMask(None=@n77) {r=true}
  n79 = LinearInterpolate(A=Multiply(A=ComponentMask(None=@n73) {b=true}, B=LinearInterpolate(Alpha=@n78) {constA=2.5}) {constB=1.5}, Alpha=@n78)
  n80 = NamedRerouteUsage()
  n81 = LinearInterpolate(A=NamedRerouteUsage(), B=@n77, Alpha=Saturate(None=Multiply(A=@n80) {constB=2}))
  n82 = Abs(None=ConstantBiasScale(None=ComponentMask(None=@n81) {g=true}) {bias=-0.5, scale=2})
  n83 = StaticSwitchParameter(
   True=ParticleColor().A
   False=LinearInterpolate(A=ScalarParameter() {defaultValue=1, parameterName=Erosion Width Exponent}, Alpha=NamedRerouteUsage()) {constA=1.6, constB=5}
  ) {parameterName=PolarCoordinate}
  n84 = ComponentMask(None=@n77) {r=true}
  n85 = OneMinus(None=@n84)
  n86 = ScalarParameter() {defaultValue=0.333, parameterName=Erosion Length Exponent}
  n87 = Subtract(A=Subtract(A=Multiply(A=@n72, B=@n79), B=Power(Base=@n82, Exp=@n83)), B=OneMinus(None=Power(Base=@n85, Exp=@n86)))
  n88 = ScalarParameter() {defaultValue=0.05, parameterName=Erosion Base Size}
  n89 = OneMinus(None=Saturate(None=Divide(A=@n84, B=@n88)))
  n90 = ScalarParameter() {defaultValue=0.5, parameterName=Dark Edge Size Tip}
  n91 = ScalarParameter() {defaultValue=1, parameterName=Dark Edge Size Base}
  n92 = Saturate(None=Power(Base=@n85, Exp=ScalarParameter() {defaultValue=0.5, parameterName=Dark Edge Transition Exponent}))
  n93 = Saturate(None=Divide(A=Subtract(A=@n87, B=@n89), B=LinearInterpolate(A=@n90, B=@n91, Alpha=@n92)))
  n94 = ScalarParameter() {defaultValue=0.2, parameterName=Opacity Softness}
  NamedRerouteDeclaration(None=Saturate(None=Divide(A=@n93, B=@n94))) {name=Dissolve_New, nodeColor=<rgba 1 0 0.6328 1>}
  n95 = ConstantBiasScale(None=@n93) {bias=-0.5, scale=3}
  n96 = ScalarParameter() {defaultValue=2, parameterName=Dark Edge Exponent}
  NamedRerouteDeclaration(None=Saturate(None=Power(Base=@n95, Exp=@n96))) {name=Dissolve_Edge_New, nodeColor=<rgba 0 0.9531 1 1>}
  n97 = LinearInterpolate(
   A=VectorParameter() {parameterName=Dark Edge, defaultValue=<rgba 0.025 0.025 0.025 0>}.RGB
   B=VectorParameter() {parameterName=Dark Edge FE, defaultValue=<rgba 0.0111 0.0121 0.015 1>}.RGB
   Alpha=ScalarParameter() {parameterName=IsFrontend}
  )
  n98 = NamedRerouteUsage()
  n99 = Saturate(None=Multiply(A=ComponentMask(None=@n98) {b=true}, B=ComponentMask(None=@n98) {r=true}))
  n100 = LinearInterpolate(
   A=LinearInterpolate(A=Multiply(A=@n97) {constB=0.8}, B=@n97, Alpha=ConstantBiasScale(None=@n99) {bias=-0.5, scale=2}) {constA=0.02, constB=0.02}
   B=NamedRerouteUsage()
   Alpha=Saturate(None=NamedRerouteUsage())
  )
  n101 = NamedRerouteUsage()
  NamedRerouteDeclaration(None=MaterialFunctionCall(Emissive=Multiply(A=@n100, B=@n101)) {materialFunction=&MF_FX_EmissiveAdjustments}.Result) {name=Emissive, nodeColor=<rgba 1 0.7214 0 1>}
  n102 = NamedRerouteUsage()
  n103 = Multiply(A=Multiply(A=ComponentMask(None=@n102) {a=true}, B=NamedRerouteUsage()), B=NamedRerouteUsage())
  n104 = TextureCoordinate()
  n105 = Multiply(A=ComponentMask(None=@n104) {r=true}, B=OneMinus(None=ComponentMask(None=@n104) {r=true}))
  n106 = Multiply(A=Power(Base=Saturate(None=@n105)) {constExponent=1.25}) {constB=10}
  NamedRerouteDeclaration(None=StaticSwitchParameter(True=Multiply(A=@n103, B=@n106), False=@n103) {parameterName=PolarCoordinate}) {name=Opacity, nodeColor=<rgba 1 0.0703 0 1>}
  NamedRerouteDeclaration(None=LinearInterpolate(
    A=DynamicParameter() {defaultValue=<rgba 1 0 0 0.5>, paramNames [Timer, Offset, InverseHand, Sprite_Offset]}.Timer
    B=Time()
    Alpha=ScalarParameter() {parameterName=TimerDebug}
   )) {name=Time, nodeColor=<rgba 0.4062 1 0 1>}
  NamedRerouteDeclaration(None=@n19.RGBA) {name=Mask, nodeColor=<rgba 0.7109 0 1 1>}
  n107 = NamedRerouteUsage()
  n108 = TextureCoordinate()
  n109 = Add(A=@n108, B=Constant2Vector() {g=-0.5})
  n110 = Add(
   A=StaticSwitchParameter(True=MaterialFunctionCall(UV Coord=@n109) {materialFunction=&MF_PolarCoordinates}.Result, False=@n108) {parameterName=PolarCoordinate}
   B=AppendVector(A=Constant() {r=0.5}, B=Constant())
  )
  n111 = ComponentMask(None=Panner(Coordinate=@n110) {speedX=-1}) {r=true}
  n112 = Sine(None=ComponentMask(None=Panner(Coordinate=@n110) {speedX=-0.5}) {r=true}) {period=0.5}
  n113 = ComponentMask(None=Panner(Coordinate=@n110) {speedX=-0.75}) {r=true}
  n114 = Add(A=Add(A=Sine(None=@n111) {period=1.4}, B=Multiply(A=@n112) {constB=0.3}), B=Multiply(A=Sine(None=@n113) {period=3}) {constB=0.7})
  n115 = Transform(None=VectorParameter() {group=WPO, parameterName=WPO Direction, defaultValue=<rgba 1 1 0 1>}.RGB) {transformSourceType=TRANSFORMSOURCE_Instance}
  n116 = TextureCoordinate()
  n117 = Add(A=@n116, B=Constant2Vector() {g=-0.5})
  n118 = ComponentMask(None=StaticSwitchParameter(True=MaterialFunctionCall(UV Coord=@n117) {materialFunction=&MF_PolarCoordinates}.Result, False=@n116) {parameterName=PolarCoordinate}) {r=true}
  n119 = Add(
   A=Multiply(
    A=MaterialFunctionCall() {materialFunction=&CameraDirectionVector}.Result
    B=ScalarParameter() {parameterName=Camera offset}
   ) {constB=0}
   B=Multiply(A=Multiply(A=@n114, B=@n115), B=Power(Base=@n118) {constExponent=1})
  )
  n120 = StaticSwitchParameter(True=Multiply(A=@n119) {constB=25}, False=Multiply(A=@n119, B=ScalarParameter() {defaultValue=10, parameterName=WPO})) {parameterName=PolarCoordinate}
  NamedRerouteDeclaration(None=StaticSwitchParameter(True=Add(A=@n107, B=@n120), False=@n120) {parameterName=InterpolatedWPO}) {name=WPO, nodeColor=<rgba 0.0781 0 1 1>}
  n121 = Constant() {r=10}
  n122 = Constant() {r=2.5}
  NamedRerouteDeclaration(None=StaticSwitchParameter(
    True=MaterialFunctionCall(Brightness=@n121, Exponent=@n122) {materialFunction=&MF_Fresnel_FX}.Fresnel Double Sided
    False=Constant() {r=1}
   ) {parameterName=PolarCoordinate}) {name=Fresnel, nodeColor=<rgba 0 0.0391 1 1>}
  n123 = VectorParameter() {parameterName=DistortionUVControl, defaultValue=<rgba 0.5 0.5 -0.4 0>}
  NamedRerouteDeclaration(None=MaterialFunctionCall(UVInput=NamedRerouteUsage(), TimeInput=NamedRerouteUsage(), Control_RGB=@n123.RGB, Control_A=@n123.RGB) {materialFunction=&FXCommon_QuickUVControl}.UV) {name=UV_Distortion, nodeColor=<rgba 0 1 0.3672 1>}
  n124 = TextureObjectParameter() {
   group=NoiseA
   parameterName=NoiseC_Texture
   samplerType=SAMPLERTYPE_LinearGrayscale
   sortPriority=0
   texture=&T_SimplePerlin_Greyscale
  }
  n125 = VectorParameter() {group=NoiseA, parameterName=NoiseC_UVControl, channelNames {a=VPan, b=UPan, g=VScale, r=UScale}, defaultValue=<rgba 0.25 1 -0.5 0>}
  n126 = NamedRerouteUsage()
  n127 = MaterialFunctionCall(
   Texture=@n124
   AdditionalUVOffset=Constant2Vector()
   UVControl_RGB=@n125
   UVControl_A=@n125
   "[Override]TimeInput"=NamedRerouteUsage()
   "[Override]UVInput"=@n126
  ) {materialFunction=&FXCommon_TextureSample}
  n128 = VectorParameter() {
   group=NoiseA
   parameterName=NoiseC_NoiseControl
   sortPriority=60
   channelNames {a=VStrength, b=UStrength, g=VOffset, r=UOffset}
   defaultValue=<rgba -0.5 -0.5 0 0.5>
  }
  n129 = TextureObjectParameter() {group=NoiseB, parameterName=NoiseB_Texture, sortPriority=0, texture=/Game/Effects/Fort_Effects/Emotes/UltraLight/Texture/T_SoftNoise_BC7}
  n130 = VectorParameter() {group=NoiseB, parameterName=NoiseB_UVControl, channelNames {a=VPan, b=UPan, g=VScale, r=UScale}, defaultValue=<rgba 1 1 -0.75 0>}
  n131 = MaterialFunctionCall(
   Texture=@n129
   AdditionalUVOffset=Constant2Vector()
   UVControl_RGB=@n130
   UVControl_A=@n130
   "[Override]TimeInput"=NamedRerouteUsage()
   "[Override]UVInput"=@n126
  ) {materialFunction=&FXCommon_TextureSample}
  n132 = VectorParameter() {
   group=NoiseB
   parameterName=NoiseB_NoiseControl
   sortPriority=60
   channelNames {a=VStrength, b=UStrength, g=VOffset, r=UOffset}
   defaultValue=<rgba -0.5 -0.5 0 0.2>
  }
  NamedRerouteDeclaration(None=Add(
    A=MaterialFunctionCall(NoiseValue=@n127.RG, Control_RGB=@n128, Control_A=@n128) {materialFunction=&FXCommon_NoiseControl}.NoiseValue
    B=MaterialFunctionCall(NoiseValue=@n131.RG, Control_RGB=@n132, Control_A=@n132) {materialFunction=&FXCommon_NoiseControl}.NoiseValue
   )) {name=NoiseOutput, nodeColor=<rgba 0.7578 0 1 1>}
  n133 = TextureCoordinate()
  n134 = Constant2Vector() {g=-0.5}
  n135 = StaticSwitchParameter(
   True=MaterialFunctionCall(UV Coord=Add(A=@n133, B=@n134)) {materialFunction=&MF_PolarCoordinates}.Result
   False=@n133
  ) {parameterName=PolarCoordinate}
  n136 = ComponentMask(None=@n135) {r=true}
  n137 = OneMinus(None=ComponentMask(None=@n135) {r=true})
  NamedRerouteDeclaration(None=Saturate(None=Multiply(A=@n136, B=@n137))) {name=GradientU_Mirror, nodeColor=<rgba 1 0.4453 0 1>}
  NamedRerouteUsage()
  NamedRerouteUsage()
  NamedRerouteUsage()
 ]
}
