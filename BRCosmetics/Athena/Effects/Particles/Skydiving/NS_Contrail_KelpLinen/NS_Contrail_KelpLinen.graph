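# Niagara system for the KelpLinen skydiving contrail cosmetic, with audio-reactive ribbons, hand shapes, glows, dark bits and fluttering butterfly meshes.
# Emitters attach per partition through a shared socket reader, and PixieDust spawns off Butterfly particles via particle reads.
# System audio scale, dive alpha and owner velocity drive size, motion and orientation, with front-end preview variants.
# Colors follow time of day, and ribbon particle counts scale by quality level.
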NiagaraSystem /BRCosmetics/Athena/Effects/Particles/Skydiving/NS_Contrail_KelpLinen/NS_Contrail_KelpLinen
dependencies {
 usedDataInterfaces [
  &NiagaraDataInterfaceArrayFloat3
  &NiagaraDataInterfaceCurve
  &NiagaraDataInterfaceColorCurve
  &NiagaraDataInterfacePlatformSet
  &NiagaraDataInterfaceArrayFloat
  &NiagaraDataInterfaceVectorCurve
  &NiagaraDataInterfaceVector2DCurve
  &NiagaraDataInterfaceParticleRead
  &NiagaraDataInterfaceSocketReader
 ]
 usedDynamicInputs [
  &Subtract_Position
  /Niagara/DynamicInputs/Integer/V2/Round_Float
  &InvertBool
  &RotateVectorByQuaternion
  /Niagara/DynamicInputs/Rotation/InvertQuaternion
  &ClampFloat
  &RemapRange
  &VectorLength
  &MakeFloatFromLinearColor
 ]
 usedModules [
  &self:SystemSpawnScript.SystemScriptSource.SystemScriptGraph.NiagaraNodeAssignment_4.SetVariables_#1
  &self:SystemSpawnScript.SystemScriptSource.SystemScriptGraph.NiagaraNodeAssignment_9.SetVariables_#2
  &SystemState
  /Game/Effects/Niagara/Modules/Skydiving/FN_DivingAttributes/v5/FN_DivingAttributes
  &self:SystemSpawnScript.SystemScriptSource.SystemScriptGraph.NiagaraNodeAssignment_0.SetVariables_#3
  &self:SystemSpawnScript.SystemScriptSource.SystemScriptGraph.NiagaraNodeAssignment_6.SetVariables_#4
  &self:SystemSpawnScript.SystemScriptSource.SystemScriptGraph.NiagaraNodeAssignment_7.SetVariables_#5
  &FN_Integrator
  &self:SystemSpawnScript.SystemScriptSource.SystemScriptGraph.NiagaraNodeAssignment_5.SetVariables_#6
  &self:SystemSpawnScript.SystemScriptSource.SystemScriptGraph.NiagaraNodeAssignment_2.SetVariables_#7
  &self:SystemSpawnScript.SystemScriptSource.SystemScriptGraph.NiagaraNodeAssignment_3.SetVariables_#8
  &self:SystemSpawnScript.SystemScriptSource.SystemScriptGraph.NiagaraNodeAssignment_8.SetVariables_#9
  &self:SystemSpawnScript.SystemScriptSource.SystemScriptGraph.NiagaraNodeAssignment_10.SetVariables_#10
 ]
}
emitters {
 BlackBits {
  emitterUpdateScript [
   &EmitterState()
   &SpawnPerFrame(
    Spawn:NiagaraBool=&SetBoolByFloatComparison(A=$System.AudioScale, B=0.75)
    Spawn Count:NiagaraInt32=&RandomRangeInt(Minimum=6, Maximum=12)
    Use Spawn Probability=false
   )
  ]
  particleSpawnScript [
   &InitializeParticle(
    Lifetime=&RandomRangeFloat(Minimum=0.4, Maximum=0.6)
    Position=$Engine.Owner.Position
    Position Mode=::NewEnumerator1
    Sprite Size=&Multiply_Vector2(A=&RandomRangeVector2D(Minimum=<2 2>, Maximum=<6 6>), B=<0.5 0.75>)
    Sprite Size Mode=::NewEnumerator4
    UsePositionOffset=false
    Write Material Random=true
   )
   &PartitionParticles(Write Partition to Ribbon ID=::NewEnumerator0)
   &SocketLocation(
    Interpolate Socket Sampling=true
    SamplePosOffset=true
    Sampled Position Offset=&FN_FrontEndVector3Modifier(
     Gameplay=<0 0 0>
     Front End=&Lerp_Vector(
      A=&SelectVectorFromArray(
       Direct Array Index=$Particles.PartitionParticles.Partition
       Vector Selection Array=DataInterface() {propertyValues {FloatData [<-10 0 0>, <10 0 0>, <10 0 0>, <-10 0 0>]}}
      )
      B=&SelectVectorFromArray(
       Direct Array Index=$Particles.PartitionParticles.Partition
       Vector Selection Array=DataInterface() {propertyValues {FloatData [<-10 0 0>, <10 0 0>, <15 0 0>, <-15 0 0>]}}
      )
      Alpha=$System.Dive
     )
    )
    Socket Index=$Particles.PartitionParticles.Partition
    Socket Reader=$System.Socket Reader
    Socket Sampling Mode=::NewEnumerator3
    VelocitySpeedLimit=false
   )
   &SphereLocation(
    MaskSpawnGroup=false
    Offset=<0 0 10>
    Override Seed=false
    Sphere Origin:NiagaraPosition=$Output.SocketLocation.SampledPosition
    Sphere Radius=10
    Surface Expansion Mode=::NewEnumerator1
    Surface Only=true
    Surface Only Band Thickness=1
    UseNonUniformScale=false
    UseOffset=true
    UseSphereOrientation=false
   ) {module=SphereLocation001}
   &CameraOffset(Camera Offset Amount=&RandomRangeFloat(Minimum=2, Maximum=5))
   &AddVelocityFromPoint(
    Module.Use Falloff Distance=false
    Origin Offset=&FN_FrontEndVector3Modifier(Gameplay=<0 0 10>, Front End=<0 50 10>)
    Velocity Strength=&UniformRangedFloat(Minimum=111, Maximum=333)
   )
  ]
  particleUpdateScript [
   &InheritSourceMovement(ApplyForce=false, ApplyPosition=true, ApplyPositionFalloff=false)
   &CurlNoiseForce(
    Noise Frequency=22
    Noise Strength=&Multiply_Float(
     A=500
     B=&FloatFromCurve(FloatCurve=DataInterface() {propertyValues {Curve {keys [<at 0 0>,<at 0.1 1>,<at 1 1>]}}})
    )
    Pan Noise=false
   )
   &Drag(Drag=&UniformRangedFloat(Minimum=4, Maximum=8), Use Linear Drag=true, Use Rotational Drag=false)
   &SolveForcesAndVelocity(Clamp Velocity=false, Limit Acceleration=false)
   &ParticleState()
   &ScaleColor(
    Scale Alpha=&FloatFromCurve(FloatCurve=DataInterface() {propertyValues {Curve {keys [
         {
          arriveTangentWeight=0.9254
          interpMode=RCIM_Cubic
          leaveTangentWeight=0.9254
          tangentMode=RCTM_User
          tangentWeightMode=RCTWM_WeightedBoth
          time=0
          value=1
         }
         {
          arriveTangent=-7.1278
          arriveTangentWeight=1.1179
          interpMode=RCIM_Cubic
          leaveTangent=-7.1278
          leaveTangentWeight=1.1179
          tangentMode=RCTM_User
          tangentWeightMode=RCTWM_WeightedBoth
          time=1
          value=0
         }
        ]}}})
    ScaleA=true
    ScaleRGB=true
   )
   &ScaleSpriteSize(
    Non-Uniform Curve Scale=&Vector2DFromFloat(Value=&FloatFromCurve(
      FloatCurve=DataInterface() {propertyValues {Curve {keys [
          {interpMode=RCIM_Cubic, tangentMode=RCTM_User, time=0, value=1}
          {arriveTangent=11.8959, interpMode=RCIM_Cubic, leaveTangent=11.8959, tangentMode=RCTM_User, time=1, value=5}
         ]}}}
      CurveIndex=&NormalizeFloat(Value=$System.SmoothedAudioScale, Min=0, Max=1)
     ))
    Non-Uniform Curve Sprite Scale=DataInterface() {propertyValues {
      XCurve {keys [<at 0.5 1 RCIM_Cubic>, {arriveTangent=-4.9638, interpMode=RCIM_Cubic, leaveTangent=-4.9638, tangentMode=RCTM_User, time=1, value=0}]}
      YCurve {keys [<at 0.5 1 RCIM_Cubic>, {arriveTangent=-1.8921, interpMode=RCIM_Cubic, leaveTangent=-1.8921, tangentMode=RCTM_User, time=1, value=0.5}]}
     }}
    Scale Sprite Size Mode=::NewEnumerator3
   )
  ]
  rendererClasses [&NiagaraSpriteRendererProperties]
  renderers [{rendererClass=&NiagaraSpriteRendererProperties}]
 }
 Butterfly {
  emitterSpawnScript [/SetVariables(Emitter.InterpolatedVector:Vector3f=<0 0 1>, Emitter.TargetVector:Vector3f=<0 0 1>) {isSetParameters=true}]
  emitterUpdateScript [
   &EmitterState()
   &SpawnRate(SpawnRate=2, Use Spawn Probability=false)
   &self:ScratchModule_08(Rate=2, Target Vector:Vector3f=$System.LookAtVector) {module=ScratchModule_08}
  ]
  particleSpawnScript [
   &PartitionParticles(Write Partition to Ribbon ID=::NewEnumerator0)
   &InitializeParticle(
    Color=&MultiplyLinearColor(Linear Color=<rgba 3 3 3 1>, Scale Factor=<1 1 1 1>)
    Lifetime=&RandomRangeFloat(Minimum=4, Maximum=3)
    Mesh Scale Mode=::NewEnumerator3
    Mesh Uniform Scale=&Multiply_Float(
     A=&RandomRangeFloat(Minimum=0.025, Maximum=0.035)
     B=&FN_FrontEndFloatModifier(Gameplay=1, Front End=0.75)
    )
    Sprite Rotation Mode=::NewEnumerator2
    Sprite Size Mode=::NewEnumerator3
    Uniform Sprite Size=30
    UsePositionOffset=false
    Write Material Random=true
   )
   &SocketLocation(
    Orientation Sampling=::NewEnumerator1
    SamplePosOffset=false
    Sampled Orientation Coordinate Space:ENiagaraCoordinateSpace=::World
    Socket Index=$Particles.PartitionParticles.Partition
    Socket Reader=$System.Socket Reader
    Socket Sampling Mode=::NewEnumerator3
    VelocitySpeedLimit=false
   )
   &ShapeLocation(
    MaskSpawnGroup=false
    Offset=&Lerp_Vector(
     A=&SelectVectorFromArray(
      Direct Array Index=$Particles.PartitionParticles.Partition
      Vector Selection Array=DataInterface() {propertyValues {FloatData [<50 45 60>, <-50 45 60>]}}
     )
     B=&TransformVector(
      Vector=&SelectVectorFromArray(
       Direct Array Index=$Particles.PartitionParticles.Partition
       Vector Selection Array=DataInterface() {propertyValues {FloatData [<75 90 20>, <-75 90 20>]}}
      )
      Destination Space=::Local
     )
     Alpha=&NormalizeFloat(Value=$System.ContrailDiveAlpha, Min=0, Max=1)
    )
    Offset Coordinate Space=::Local
    Offset Mode=::NewEnumerator0
    Override Seed=true
    Shape Origin=$Output.SocketLocation.SampledPosition
    Sphere Radius=&Multiply_Float(
     A=&FN_FrontEndFloatModifier(Gameplay=1, Front End=0)
     B=&Lerp_Float(A=25, B=50, Alpha=&NormalizeFloat(Value=$System.ContrailDiveAlpha, Min=0, Max=1))
    )
    UseOwnerScale=true
   )
   &InitialMeshOrientation(Use Rotation Vector=false)
   /SetVariables(Particles.Position=&FN_FrontEndVector3Modifier(
     Gameplay=Linked() {linkedVariable {name=Particles.Position, type=NiagaraPosition}}
     Front End=&Lerp_Vector(
      A=&RandomRangeVector(Minimum=<-100 50 100>, Maximum=<100 50 150>)
      B=&RandomRangeVector(Minimum=<-100 -100 100>, Maximum=<100 0 150>)
      Alpha=$System.Dive
     )
    )) {isSetParameters=true}
   &AddVelocity(
    Velocity=&FN_FrontEndVector3Modifier(Gameplay=<0 0 1>, Front End=&Lerp_Vector(A=<0 -0.2 1>, B=<0 -1 1>, Alpha=$System.Dive))
    Velocity Speed Scale=&RandomRangeFloat(Minimum=25, Maximum=50)
   )
   /SetVariables() {isSetParameters=true}
  ]
  particleUpdateScript [
   &ParticleState()
   &SpriteRotationRate(Rotation Rate=&RandomRangeFloat(Minimum=-5, Maximum=5))
   &CurlNoiseForce(
    Noise Frequency=5
    Noise Strength=&FN_FrontEndFloatModifier(Gameplay=100, Front End=25)
    Pan Noise=true
    Pan Noise Field=<1 0 0>
   )
   &SolveForcesAndVelocity(Clamp Velocity=false, Limit Acceleration=false)
   &ScaleMeshSize(Scale Factor=&VectorFromFloat(Value=&FloatFromCurve(
      FloatCurve=DataInterface() {propertyValues {Curve {keys [
          <at 0 0 RCIM_Cubic>
          {arriveTangent=1.5, interpMode=RCIM_Cubic, leaveTangent=1.5, time=0.2, value=0.55}
          {arriveTangent=-0.5, interpMode=RCIM_Cubic, leaveTangent=-0.5, time=0.7, value=1.05}
          {arriveTangent=-6.0964, interpMode=RCIM_Cubic, leaveTangent=-6.0964, tangentMode=RCTM_User, time=1, value=0.15}
         ]}}}
      Scale Curve=&Lerp_Float(A=1, B=1.5, Alpha=$System.SmoothedAudioScale)
     )))
   &ScaleColor(
    Linear Color Curve=DataInterface() {propertyValues {
      AlphaCurve {keys [<at 0 0>,<at 0.1184 1>,<at 0.9076 0>]}
      BlueCurve {keys [<at 0 1>,<at 1 1>]}
      GreenCurve {keys [<at 0 1>,<at 1 1>]}
      RedCurve {keys [<at 0 1>,<at 1 1>]}
     }}
    Scale Mode=::NewEnumerator2
   )
   &ScaleColor(
    Scale Mode=::NewEnumerator1
    Scale RGBA=&MakeVector4_FromLinearColor(LinearColor=&FNTimeOfDayColor(
      Night Color=<rgba 0 0.35 1 1>
      Morning Color=<rgba 0.5217 0.1457 1.5 1>
      Day Color=<rgba 0 0.525 1.5 1>
      Evening Color=<rgba 0.5394 0.1402 1.5 1>
     ))
    ScaleRGBA=true
   ) {module=ScaleColor001}
   /SetVariables(Particles.FlapTime:NiagaraFloat=&Add_Float(
     A=$Particles.FlapTime
     B=&Multiply_Float(
      A=$Engine.DeltaTime
      B=&FloatFromCurve(
       FloatCurve=DataInterface() {propertyValues {Curve {keys [{arriveTangent=0.121, interpMode=RCIM_Cubic, leaveTangent=0.121, tangentMode=RCTM_User, time=0, value=0.875}, <at 1 2 RCIM_Cubic>]}}}
       CurveIndex=&NormalizeFloat(Value=$System.ContrailDiveAlpha, Min=0, Max=1)
      )
     )
    )) {isSetParameters=true}
   /SetVariables(Particles.MeshOrientation=&LerpQuaternion(
     Quaternion A:Quat4f=&MultiplyQuaternion(
      Quaternion A=&MultiplyQuaternion(
       Quaternion A=&MakeQuatFromAxisAngle(
        Axis=&RandomRangeVector(Minimum=<0 -1 0>, Maximum=<0 1 0>)
        AngleInDegrees=&Waveform(Global Phase=$Particles.Age, "Waveform [1]"=::NewEnumerator10, "[1] Frequency"=1, "[1] Amplitude Min/Max"=<50 150>)
       )
       Quaternion B=&MakeQuatFromAxisAngle(
        Axis=<1 0 0>
        AngleInDegrees=&Waveform(Global Phase=$Particles.Age, "Waveform [1]"=::NewEnumerator10, "[1] Frequency"=2.5, "[1] Amplitude Min/Max"=<-60 -130>)
       )
      )
      Quaternion B=&MakeQuatFromAxisAngle(
       Axis=<0 1 0>
       AngleInDegrees=&Waveform(Global Phase=$Particles.Age, "Waveform [1]"=::NewEnumerator10, "[1] Frequency"=2, "[1] Amplitude Min/Max"=<-30 -60>)
      )
     )
     Quaternion B:Quat4f=&FindQuatBetweenTwoVectors(
      From Vector=&FN_FrontEndVector3Modifier(Gameplay=<0 0 1>, Front End=&Lerp_Vector(A=<0 0 1>, B=<0 -1 -1>, Alpha=$System.Dive))
      To Vector=&Multiply_Vector(
       A=&NormalizeVector(
        Vector To Normalize=&TransformVector(
         Vector=&Multiply_Vector(A=$Engine.Owner.Velocity, B=<1 1 1>)
         Source Space=::World
         Destination Space=::World
         Apply Scale=true
        )
        Safe Normalize=true
        Safe Fallback Threshold=0
        Safe Fallback Vector=<0 0 0.001>
       )
       B=<1 1 1>
      )
     )
     Lerp Factor:NiagaraFloat=&Lerp_Float(
      A=0
      B=1
      Alpha=&Add_Float(
       A=&MakeCustomFloatFromBool(
        Bool=&Boolean_AndOperation(
         Bool1=&SetBoolByFloatComparison(
          A=&Abs_Float(Float:NiagaraFloat=&MakeFloatFromVector(Vector=$Engine.Owner.Velocity, Channel=::NewEnumerator0))
          B=100
         )
         Bool2=&SetBoolByFloatComparison(
          A=&Abs_Float(Float:NiagaraFloat=&MakeFloatFromVector(Vector=$Engine.Owner.Velocity, Channel=::NewEnumerator1))
          B=100
         )
        )
        True Float=1
        False Float=0
       )
       B=$System.ContrailDiveAlpha
      )
     )
     Enable Spherical Linear Interpolation:NiagaraBool=false
    )) {isSetParameters=true}
   &InterpolateOverTime(Rate Of Change=4, Target Quat Value:Quat4f=$Particles.MeshOrientation, Variable Type=::NewEnumerator5)
   /SetVariables(Particles.MeshOrientation=$StackContext.InterpolateOverTime.InterpolateOverTime_Quaternion.Moving Average) {isSetParameters=true}
   /SetVariables(Particles.TrailOrientation:Quat4f=&FindQuatBetweenTwoVectors(
     From Vector=&FN_FrontEndVector3Modifier(Gameplay=<0 0 1>, Front End=&Lerp_Vector(A=<0 0 1>, B=<0 1 1>, Alpha=$System.Dive))
     To Vector=&Multiply_Vector(
      A=&NormalizeVector(
       Vector To Normalize=&TransformVector(
        Vector=&Multiply_Vector(A=$Engine.Owner.Velocity, B=<-1 -1 -1>)
        Source Space=::World
        Destination Space=::Local
        Apply Scale=true
       )
       Safe Normalize=true
       Safe Fallback Threshold=0
       Safe Fallback Vector=<0 0 0.001>
      )
      B=<0.1 0.1 0.1>
     )
    )) {isSetParameters=true}
   /SetVariables(Particles.Trail_Size:Vector3f=&Multiply_VectorByFloat(
     Vector=&Multiply_Vector(
      A=<1.25 1.25 0.8>
      B=&MakeVector(
       X=1
       Y=1
       Z=&FloatFromCurve(FloatCurve=DataInterface() {propertyValues {Curve {keys [
            {tangentMode=RCTM_User, time=0, value=1}
            {arriveTangent=-0.0405, interpMode=RCIM_Cubic, leaveTangent=-0.0405, tangentMode=RCTM_User, time=0.7009, value=1}
            {arriveTangent=-8.0089, interpMode=RCIM_Cubic, leaveTangent=-8.0089, tangentMode=RCTM_User, time=1, value=0}
           ]}}})
      )
     )
     Float=&FN_FrontEndFloatModifier(Gameplay=2, Front End=1.5)
    )) {isSetParameters=true}
   &DynamicMaterialParameters(
    Index 3 Param 1=$Particles.Age
    Index 3 Param 2=&RandomRangeFloat(Minimum=0, Maximum=1)
    Module.Index3Param0Write=true
    Module.Index3Param1Write=true
    Module.Index3Param2Write=false
    Module.Index3Param3Write=false
    Write Parameter Index 0=false
    Write Parameter Index 3=true
   )
   /SetVariables(
    Particles.DynamicMaterialParameterButteffly01:Vector4f=&MakeVector4(X=4000, Y=1, Z=0.05, W=0)
    Particles.DynamicMaterialParameterButtefly:Vector4f=&MakeVector4(
     X=&Lerp_Float(A=0.2, B=0.05, Alpha=&NormalizeFloat(Value=$System.ContrailDiveAlpha, Min=0, Max=1))
     Y=&RandomRangeFloat(Minimum=0, Maximum=1)
     Z=&Lerp_Float(A=0.2, B=0.05, Alpha=&NormalizeFloat(Value=$System.ContrailDiveAlpha, Min=0, Max=1))
     W=$Particles.FlapTime
    )
    Particles.DynamicMaterialParameterGlow:Vector4f=<2 0 0 0>
   ) {isSetParameters=true}
   /SetVariables(Particles.Color_Trail:LinearColor=&MultiplyLinearColor(
     Linear Color=&ColorFromCurve(ColorCurve=DataInterface() {propertyValues {
        AlphaCurve {keys [<at 0.0053 0>,<at 0.4728 0.5 RCIM_Cubic>,<at 1 0 RCIM_Cubic>]}
        BlueCurve {keys [<at 0 1 RCIM_Cubic>,<at 1 1 RCIM_Cubic>]}
        GreenCurve {keys [<at 0 1 RCIM_Cubic>,<at 1 1 RCIM_Cubic>]}
        RedCurve {keys [<at 0 1 RCIM_Cubic>,<at 1 1 RCIM_Cubic>]}
       }})
     Scale Factor=&MakeFloat4FromVectorAndFloat(
      "Vector (RGB)"=&MakeVectorFromLinearColorRGB(Color=&FNTimeOfDayColor(Night Color=<rgba 1.1586 1.125 1.5 1>, Morning Color=<rgba 0.8919 0.75 1 1>, Day Color=<rgba 1 1 1 1>, Evening Color=<rgba 0.7699 0.6 0.8 1>))
      "Float (Alpha)"=&FN_FrontEndFloatModifier(Gameplay=2, Front End=0.75)
     )
    )) {isSetParameters=true}
   /SetVariables(Particles.Color_Glow:LinearColor=&MultiplyLinearColor(
     Linear Color=&MultiplyLinearColor(
      Linear Color=$Particles.Color
      Scale Factor=&MakeVector4_FromLinearColor(LinearColor=&FNTimeOfDayColor(
        Night Color=<rgba 0 0.4375 1.25 1>
        Morning Color=<rgba 0.6615 0.2168 1 1>
        Day Color=<rgba 0 0.4375 1.25 1>
        Evening Color=<rgba 0.7021 0.1485 0.8 1>
       ))
     )
     Scale Factor=&MakeFloat4FromVectorAndFloat(
      "Vector (RGB)"=&VectorFromFloat(Value=0.25)
      "Float (Alpha)"=&Multiply_Float(
       A=&FN_FrontEndFloatModifier(Gameplay=1, Front End=0.25)
       B=&Lerp_Float(A=0.3, B=5, Alpha=&NormalizeFloat(Value=$System.SmoothedAudioScale, Min=0, Max=1))
      )
     )
    )) {isSetParameters=true}
  ]
  rendererClasses [&NiagaraMeshRendererProperties, &NiagaraSpriteRendererProperties]
  renderers [
   {rendererClass=&NiagaraMeshRendererProperties, rendererIndex=0}
   {rendererClass=&NiagaraMeshRendererProperties, rendererIndex=1}
   {rendererClass=&NiagaraSpriteRendererProperties, rendererIndex=2}
  ]
 }
 Decal_Glow {
  emitterUpdateScript [
   &EmitterState(Life Cycle Mode=::NewEnumerator1, Loop Behavior=::NewEnumerator1, Loop Duration Mode=::NewEnumerator1, UseLoopDelay=false)
   &SpawnBurst_Instantaneous(Spawn Count=4, Use Spawn Probability=false, UseLoopCountLimit=false)
  ]
  particleSpawnScript [
   &InitializeParticle(
    Lifetime=&RandomRangeFloat(Minimum=1, Maximum=2)
    Sprite Rotation Mode=::NewEnumerator2
    Sprite Size Mode=::NewEnumerator3
    Sprite UV Mode=::NewEnumerator3
    Uniform Sprite Size=&RandomRangeFloat(Minimum=75, Maximum=100)
    UsePositionOffset=false
    Write Material Random=true
   )
   &PartitionParticles(Number Of Partitions=4, Write Partition to Ribbon ID=::NewEnumerator0)
  ]
  particleUpdateScript [
   &ParticleState(
    Kill Particles When Lifetime Has Elapsed=false
    Let Infinitely Lived Particles Die When Emitter Deactivates=true
    Loop Particles Lifetime=true
   )
   &SocketLocation(
    SamplePosOffset=false
    Socket Index=$Particles.PartitionParticles.Partition
    Socket Reader=$System.Socket Reader
    Socket Sampling Mode=::NewEnumerator3
    VelocitySpeedLimit=false
   )
   &CameraOffset(Camera Offset Amount=10)
   &ScaleColor(
    Scale Alpha=&FloatFromCurve(
     FloatCurve=DataInterface() {propertyValues {Curve {keys [
         {arriveTangent=-5.2731, interpMode=RCIM_Cubic, leaveTangent=-5.2731, tangentMode=RCTM_User, time=0, value=1}
         {arriveTangent=44.865, interpMode=RCIM_Cubic, leaveTangent=44.865, tangentMode=RCTM_User, time=1, value=25}
        ]}}}
     CurveIndex=&NormalizeFloat(Value=$System.SmoothedAudioScale, Min=0, Max=1)
    )
    Scale RGB=&Multiply_Vector(
     A=&MakeVectorFromLinearColorRGB(Color=&FN_Select_Color_from_Time_Of_Day(
       Use Manual Time of Day Thresholds=true
       Morning=<rgba 0.1641 0.1118 1 1>
       Day=<rgba 0 0.4392 1 1>
       Evening=<rgba 0.5714 0.4105 1 1>
       Night=<rgba 0.3255 0.6196 0.902 1>
       Evening Threshold=22
      ))
     B=&VectorFromFloat(Value=&FloatFromCurve(
       FloatCurve=DataInterface() {propertyValues {Curve {keys [
           {interpMode=RCIM_Cubic, tangentMode=RCTM_User, time=0, value=1}
           {arriveTangent=5.91, interpMode=RCIM_Cubic, leaveTangent=5.91, tangentMode=RCTM_User, time=1, value=5}
          ]}}}
       CurveIndex=&NormalizeFloat(Value=$System.SmoothedAudioScale, Min=0, Max=1)
      ))
    )
    ScaleA=true
    ScaleRGB=true
   )
   &ScaleSpriteSize(Scale Factor=&Vector2DFromFloat(Value=&Multiply_Float(
      A=&FloatFromCurve(
       FloatCurve=DataInterface() {propertyValues {Curve {keys [
           {interpMode=RCIM_Cubic, tangentMode=RCTM_User, time=0, value=1}
           {interpMode=RCIM_Cubic, tangentMode=RCTM_User, time=0.5, value=1.1}
           {tangentMode=RCTM_User, time=1, value=1}
          ]}}}
       Scale Curve=&SelectFloatFromArray(
        Direct Array Index=$Particles.PartitionParticles.Partition
        Float Selection Array=DataInterface() {propertyValues {FloatData [0.8,0.8,0.4,0.4]}}
       )
      )
      B=&Lerp_Float(A=1, B=1.5, Alpha=$System.SmoothedAudioScale)
     )))
   &FN_Sprite_Attributes(
    Color=&MultiplyLinearColor(
     Linear Color=&MultiplyLinearColor(
      Linear Color=<rgba 1 1 1 1>
      Scale Factor=&MakeFloat4FromVectorAndFloat("Vector (RGB)"=<1 1 1>, "Float (Alpha)"=&FN_FrontEndFloatModifier(Gameplay=1, Front End=0.5))
     )
     Scale Factor=&MakeVector4_FromLinearColor(LinearColor=&FN_Select_Color_from_Time_Of_Day(
       Use Manual Time of Day Thresholds=true
       Morning=<rgba 0.5064 0.5099 2 0.25>
       Day=<rgba 0.25 2.2292 5 0.25>
       Evening=<rgba 0.4722 0.6472 2 0.25>
       Night=<rgba 0.3255 0.6196 0.902 1>
       Evening Threshold=22
      ))
    )
    Create Color Attribute=true
    Create SpriteSize Attribute=true
    SpriteSize=&Multiply_Vector2D(A=$Particles.SpriteSize, B=<1.5 1.5>)
    Write to Module Namespace=true
   )
   &DynamicMaterialParameters(
    Index 0 Param 1=&Lerp_Float(A=1.5, B=0, Alpha=$System.SmoothedAudioScale)
    Index 0 Param 3=5
    Index 0 Param 4=&RandomRangeFloat(Minimum=0, Maximum=1)
    Index 2 Param 1=3
    Index 2 Param 2=1
    Index 2 Param 3=1
    Index 2 Param 4=1
    Module.Param0WriteEnabled=true
    Module.Param0WriteEnabled2=true
    Module.Param1WriteEnabled=true
    Module.Param1WriteEnabled2=true
    Module.Param2WriteEnabled=true
    Module.Param2WriteEnabled2=true
    Module.Param3WriteEnabled=true
    Module.Param3WriteEnabled2=true
    Write Parameter Index 2=true
   )
  ]
  rendererClasses [&NiagaraSpriteRendererProperties]
  renderers [
   {rendererClass=&NiagaraSpriteRendererProperties}
   {rendererClass=&NiagaraSpriteRendererProperties, rendererIndex=1}
  ]
 }
 FN_Debug {
  bEnabled=false
  emitterUpdateScript [
   &EmitterState()
   &SpawnBurst_Instantaneous(Spawn Count=1, Use Spawn Probability=false)
  ]
  particleSpawnScript [
   /Niagara/Modules/Spawn/Initialization/InitializeParticle(
    Lifetime=5
    Module.Write Color=true
    Module.Write Lifetime=true
    Module.Write Mass=true
    Module.Write Material Random=true
    Module.Write Position=false
    Module.Write Scale=false
    Module.Write SpriteRotation=false
    Module.Write SpriteSize=true
    Position=&SimulationPosition()
    Sprite Size=<100 100>
   )
   &SystemLocation(MaskSpawnGroup=false, Offset=<0 0 150>)
  ]
  particleUpdateScript [
   &ParticleState()
   /SetVariables(Particles.Debug:NiagaraFloat=$System.AudioScale) {isSetParameters=true}
   &DynamicMaterialParameters(
    Index 0 Param 1=$Particles.Debug
    Module.Param0WriteEnabled=true
    Module.Param1WriteEnabled=true
    Module.Param2WriteEnabled=true
    Module.Param3WriteEnabled=true
   )
  ]
  rendererClasses [&NiagaraSpriteRendererProperties]
  renderers [{rendererClass=&NiagaraSpriteRendererProperties}]
 }
 Hands_Support_Shapes {
  emitterSpawnScript [
   /SetVariables(Emitter.InterpolatedVector:Vector3f=<0 0 1>, Emitter.TargetVector:Vector3f=<0 0 1>) {isSetParameters=true}
   /SetVariables(Emitter.isFrontend:NiagaraFloat=&MakeCustomFloatFromBool(Bool=$User.bIsFrontEndPreview, True Float=1, False Float=0)) {isSetParameters=true}
  ]
  emitterUpdateScript [
   &EmitterState(
    Inactive Response=::NewEnumerator1
    Life Cycle Mode=::NewEnumerator1
    Loop Behavior=::NewEnumerator1
    Loop Duration Mode=::NewEnumerator1
    UseLoopDelay=false
   )
   &SpawnBurst_Instantaneous(Spawn Count=2, Use Spawn Probability=false, UseLoopCountLimit=false)
   &self:ScratchModule_07(Rate=2, Target Vector:Vector3f=$System.LookAtVector) {module=ScratchModule_07}
  ]
  particleSpawnScript [
   &PartitionParticles(Particles Per Partition=1, Partition Mode:ENiagara_PartitionMode=::NewEnumerator1)
   /SetVariables() {isSetParameters=true}
   &DynamicMaterialParameters(
    Index 0 Param 4=&SelectFloatFromArray(Direct Array Index=$Particles.PartitionParticles.Partition, Float Selection Array=DataInterface() {propertyValues {FloatData [0.3,0.7]}})
    Module.Index0Param0Write=false
    Module.Index0Param1Write=false
    Module.Index0Param2Write=false
    Module.Index0Param3Write=true
   )
  ]
  particleUpdateScript [
   &SocketLocation(
    Orientation Sampling=::NewEnumerator1
    SamplePosOffset=true
    Sampled Position Coordinate Space:ENiagaraCoordinateSpace=::Local
    Sampled Position Offset Coordinate Space:ENiagaraSkelMeshTransforms=::NewEnumerator2
    Socket Index=$Particles.PartitionParticles.Partition
    Socket Reader=$System.Socket Reader
    Socket Sampling Mode=::NewEnumerator3
    VelocitySpeedLimit=false
   )
   /SetVariables(Particles.RibbonTime:NiagaraFloat=&Add_Float(
     A=$Particles.RibbonTime
     B=&Multiply_Float(
      A=$Engine.DeltaTime
      B=&FloatFromCurve(
       FloatCurve=DataInterface() {propertyValues {Curve {keys [{arriveTangent=0.121, interpMode=RCIM_Cubic, leaveTangent=0.121, tangentMode=RCTM_User, time=0, value=1}, <at 1 3 RCIM_Cubic>]}}}
       CurveIndex=&NormalizeFloat(Value=$System.ContrailDiveAlpha, Min=0, Max=1)
      )
     )
    )) {isSetParameters=true}
   /SetVariables(Particles.MeshOrientation=&MultiplyQuaternion(
     Quaternion A=&MakeQuatFromAxisAngle(
      Axis=<0 0 1>
      AngleInDegrees=&SelectFloatFromArray(Direct Array Index=$Particles.PartitionParticles.Partition, Float Selection Array=DataInterface() {propertyValues {FloatData [10,170]}})
     )
     Quaternion B=&FindQuatBetweenTwoVectors(
      From Vector=<0 0 1>
      To Vector=&Lerp_Vector(
       A=&FN_FrontEndVector3Modifier(
        Gameplay=&SelectVectorFromArray(
         Direct Array Index=$Particles.PartitionParticles.Partition
         Vector Selection Array=DataInterface() {propertyValues {FloatData [<0.7 0.1 1>, <0.7 -0.1 1>]}}
        )
        Front End=&SelectVectorFromArray(
         Direct Array Index=$Particles.PartitionParticles.Partition
         Vector Selection Array=DataInterface() {propertyValues {FloatData [<0 -0.35 1>, <0 0.35 1>]}}
        )
       )
       B=&FN_FrontEndVector3Modifier(
        Gameplay=&SelectVectorFromArray(
         Direct Array Index=$Particles.PartitionParticles.Partition
         Vector Selection Array=DataInterface() {propertyValues {FloatData [<0.5 0.8 1>, <0.5 -0.8 1>]}}
        )
        Front End=&SelectVectorFromArray(
         Direct Array Index=$Particles.PartitionParticles.Partition
         Vector Selection Array=DataInterface() {propertyValues {FloatData [<0 -1 1>, <0 1 1>]}}
        )
       )
       Alpha=&NormalizeFloat(Value=$System.ContrailDiveAlpha, Min=0, Max=1)
      )
     )
    )) {isSetParameters=true}
   /SetVariables(Particles.Scale=&Multiply_VectorByFloat(
     Vector=&Multiply_Vector(
      A=<1.2 1.2 4>
      B=&Lerp_Vector(A=<1 1 1>, B=<0.8 0.8 1.25>, Alpha=&NormalizeFloat(Value=$System.ContrailDiveAlpha, Min=0, Max=1))
     )
     Float=&Multiply_Float(
      A=&Lerp_Float(A=1, B=1.25, Alpha=$System.SmoothedAudioScale)
      B=&FN_FrontEndFloatModifier(Gameplay=1, Front End=0.8)
     )
    )) {isSetParameters=true}
   &InterpolateOverTime(
    Rate Of Change:NiagaraFloat=&Lerp_Float(A=5, B=1, Alpha=&NormalizeFloat(Value=$System.ContrailDiveAlpha, Min=0, Max=1))
    Target Position Value:NiagaraPosition=$Particles.Position
    Variable Type=::NewEnumerator6
   )
   /SetVariables(Particles.Interpolated_WPO:Vector4f=&MakeFloat4FromVectorAndFloat(
     "Vector (RGB)"=&TransformPosition(
      Position=Linked() {linkedVariable {name=StackContext.InterpolateOverTime.InterpolateOverTime_Position.Moving Average, type=Vector3f}}
      Source Space=::Local
      Destination Space=::World
     )
     "Float (Alpha)"=5
    )) {isSetParameters=true}
   &DynamicMaterialParameters(
    Index 0 Param 1=$Particles.RibbonTime
    Index 2 Param 1=$System.SmoothedAudioScale
    Module.Index0Param0Write=true
    Module.Index0Param1Write=false
    Module.Index0Param2Write=false
    Module.Index0Param3Write=false
    Module.Index2Param0Write=true
    Module.Index2Param1Write=true
    Module.Index2Param2Write=true
    Module.Index2Param3Write=true
    Write Parameter Index 2=true
   ) {module=DynamicMaterialParameters001}
  ]
  rendererClasses [&NiagaraMeshRendererProperties]
  renderers [{rendererClass=&NiagaraMeshRendererProperties, rendererIndex=0}]
 }
 PixieDust {
  emitterUpdateScript [
   &EmitterState()
   &SpawnRate(SpawnRate=20, Use Spawn Probability=false)
  ]
  particleSpawnScript [
   &FN_SetSpawnGroup(Number of Spawn Groups=10)
   /Niagara/Modules/Spawn/Initialization/V2/InitializeParticle(
    Color=&RandomRangeLinearColor(
     Minimum=&RandomRangeLinearColor(Minimum=<rgba 0.6604 0 1 0.7>, Maximum=<rgba 0.4167 0 1 0.7>)
     Maximum=&RandomRangeLinearColor(Minimum=<rgba 0 1 0.7667 0.7>, Maximum=<rgba 0 0.5833 1 0.7>)
    )
    Lifetime=&UniformRangedFloat(Minimum=0.5, Maximum=0.75)
    Sprite Rotation Mode=::NewEnumerator2
    Sprite Size Mode=::NewEnumerator3
    Uniform Sprite Size=&Multiply_Float(
     A=&RandomRangeFloat(Minimum=40, Maximum=50)
     B=&MaskFloatBySpawnGroup(Masked Float=1, Passthrough Float=&RandomRangeFloat(Minimum=0.4, Maximum=0.6))
    )
    UsePositionOffset=false
    Write Material Random=true
   )
   &self:SamplePositionFromOtherParticle(
    PAR=DataInterface() {propertyValues {EmitterBinding {bindingMode=Other, emitterName=Butterfly}}}
    Position_Offset:NiagaraPosition=&Multiply_Vector(
     A=&NormalizeVector(
      Vector To Normalize=&TransformVector(
       Vector=&Multiply_Vector(A=$Engine.Owner.Velocity, B=<-1 -1 -1>)
       Source Space=::World
       Destination Space=::Local
       Apply Scale=true
      )
      Safe Normalize=true
      Safe Fallback Threshold=0
      Safe Fallback Vector=<0 0 0.001>
     )
     B=<30 30 30>
    )
   ) {module=SamplePositionFromOtherParticle}
   &DynamicMaterialParameters(
    Index 0 Param 1=&RandomRangeFloat(Minimum=-0.05, Maximum=0.05)
    Index 0 Param 2=&RandomRangeFloat(Minimum=0, Maximum=32)
    Module.Index0Param0Write=true
    Module.Index0Param1Write=true
    Module.Index0Param2Write=false
    Module.Index0Param3Write=false
   )
   &AddVelocity(
    Velocity=&FN_FrontEndVector3Modifier(Gameplay=<0 0 0>, Front End=&Lerp_Vector(A=<0 -0.2 1>, B=<0 -1 1>, Alpha=$System.Dive))
    Velocity Speed Scale=&RandomRangeFloat(Minimum=5, Maximum=10)
   )
  ]
  particleUpdateScript [
   &ParticleState()
   &Drag(Drag=200, Use Linear Drag=true, Use Rotational Drag=true)
   &InheritSourceMovement(Applied Position Scale=<1 1 0.95>, ApplyForce=false, ApplyPosition=true, ApplyPositionFalloff=false)
   &SolveForcesAndVelocity(Clamp Velocity=false, Limit Acceleration=false)
   &ScaleColor(
    Scale Alpha=&FloatFromCurve(
     FloatCurve=DataInterface() {propertyValues {Curve {keys [
         {arriveTangent=9.0725, interpMode=RCIM_Cubic, leaveTangent=9.0725, tangentMode=RCTM_User, time=0, value=0}
         {arriveTangent=-0.5016, interpMode=RCIM_Cubic, leaveTangent=-0.5016, tangentMode=RCTM_User, time=0.1408, value=0.9847}
         {arriveTangent=-0.0405, interpMode=RCIM_Cubic, leaveTangent=-0.0405, tangentMode=RCTM_User, time=1, value=0}
        ]}}}
     Scale Curve=0.5
    )
    ScaleA=true
    ScaleRGB=true
   )
   &ScaleSpriteSize(Uniform Curve Sprite Scale=DataInterface() {propertyValues {Curve {keys [
        {interpMode=RCIM_Cubic, tangentMode=RCTM_User, time=0, value=1}
        {arriveTangent=-1.6461, interpMode=RCIM_Cubic, leaveTangent=-1.6461, tangentMode=RCTM_User, time=1, value=0.5}
       ]}}})
  ]
  rendererClasses [&NiagaraSpriteRendererProperties]
  renderers [{rendererClass=&NiagaraSpriteRendererProperties}]
 }
 Ribbons {
  emitterSpawnScript [
   /SetVariables(Emitter.InterpolatedVector:Vector3f=<0 0 1>, Emitter.TargetVector:Vector3f=<0 1 1>) {isSetParameters=true}
   /SetVariables(Emitter.RibbonLength=275) {isSetParameters=true}
   /SetVariables(Emitter.ParticlesPerTrail:NiagaraInt32=&MakeCustomIntFromBool(
     Bool=&IsPlatformSetActive(New Platform Set=DataInterface() {propertyValues {Platforms {qualityLevelMask=-8}}})
     True INT=10
     False INT=&MakeCustomIntFromBool(
      Bool=&IsPlatformSetActive(New Platform Set=DataInterface() {propertyValues {Platforms {qualityLevelMask=-28}}})
      True INT=7
      False INT=&MakeCustomIntFromBool(
       Bool=&IsPlatformSetActive(New Platform Set=DataInterface() {propertyValues {Platforms {qualityLevelMask=-30}}})
       True INT=5
       False INT=3
      )
     )
    )) {isSetParameters=true}
  ]
  emitterUpdateScript [
   &EmitterState(
    Inactive Response=::NewEnumerator1
    Life Cycle Mode=::NewEnumerator1
    Loop Behavior=::NewEnumerator1
    Loop Duration Mode=::NewEnumerator1
    UseLoopDelay=false
   )
   &SpawnBurst_Instantaneous(
    Spawn Count=&Multiply_Int(A=2, B=$Emitter.ParticlesPerTrail)
    Use Spawn Probability=false
    UseLoopCountLimit=false
   )
   &self:ScratchModule_05(Target Vector:Vector3f=$System.LookAtVector) {module=ScratchModule_05}
  ]
  particleSpawnScript [
   &PartitionParticles(
    Normalized Index In Partition Includes One=true
    Particles Per Partition:NiagaraInt32=$Emitter.ParticlesPerTrail
    Partition Mode:ENiagara_PartitionMode=::NewEnumerator1
    Write Partition Index to Ribbon Link Order:ENiagara_AttributeSamplingApplyOutput=::NewEnumerator0
    Write Partition to Ribbon ID=::NewEnumerator0
   )
   /SetVariables() {isSetParameters=true}
   &DynamicMaterialParameters(
    Index 0 Param 2=&SelectFloatFromArray(Direct Array Index=$Particles.PartitionParticles.Partition, Float Selection Array=DataInterface() {propertyValues {FloatData [0.3,0.7]}})
    Module.Index0Param0Write=false
    Module.Index0Param1Write=true
    Module.Index0Param2Write=false
    Module.Index0Param3Write=false
   )
  ]
  particleUpdateScript [
   &SocketLocation(
    Orientation Sampling=::NewEnumerator1
    Position Sampling=::NewEnumerator1
    SamplePosOffset=false
    Socket Index=$Particles.PartitionParticles.Partition
    Socket Reader=$System.Socket Reader
    Socket Sampling Mode=::NewEnumerator3
    VelocitySpeedLimit=false
   )
   /SetVariables(Particles.AudioScale:NiagaraFloat=&Lerp_Float(
     A=0.9
     B=&FloatFromCurve(
      FloatCurve=DataInterface() {propertyValues {Curve {keys [
          {interpMode=RCIM_Cubic, tangentMode=RCTM_User, time=0, value=0.8}
          {arriveTangent=-8.1804, interpMode=RCIM_Cubic, leaveTangent=-8.1804, tangentMode=RCTM_User, time=0.3, value=1.3}
          {interpMode=RCIM_Cubic, tangentMode=RCTM_User, time=0.5, value=1.4}
          {arriveTangent=2.5086, interpMode=RCIM_Cubic, leaveTangent=2.5086, tangentMode=RCTM_User, time=0.8, value=0.65}
          {tangentMode=RCTM_User, time=1, value=0.8}
         ]}}}
      CurveIndex=&ModuloFloat(A=&Add_Float(A=$Particles.PartitionParticles.Normalized Index In Partition, B=0), B=1)
      Scale Curve=&FloatFromCurve(
       FloatCurve=DataInterface() {propertyValues {Curve {keys [
           {interpMode=RCIM_Cubic, tangentMode=RCTM_User, time=0, value=1}
           {arriveTangent=-26.1951, interpMode=RCIM_Cubic, leaveTangent=-26.1951, tangentMode=RCTM_User, time=0.1, value=2.575}
           {arriveTangent=9.7614, interpMode=RCIM_Cubic, leaveTangent=9.7614, tangentMode=RCTM_User, time=0.2, value=1.8}
           {arriveTangent=1.744, interpMode=RCIM_Cubic, leaveTangent=1.744, tangentMode=RCTM_User, time=0.3, value=1.65}
           {arriveTangent=1.9058, interpMode=RCIM_Cubic, leaveTangent=1.9058, tangentMode=RCTM_User, time=0.7, value=1.125}
           {tangentMode=RCTM_User, time=1, value=1}
          ]}}}
       CurveIndex=&ModuloFloat(A=&Add_Float(A=$Particles.PartitionParticles.Normalized Index In Partition, B=0), B=1)
       Scale Curve=0.75
      )
     )
     Alpha=&Waveform(
      Global Input Position=$Particles.PartitionParticles.Normalized Index In Partition
      "Waveform [1]"=::NewEnumerator0
      "[1] Frequency"=1
      "[1] Amplitude Min/Max"=<0 1>
      "[1] Scale"=$System.AudioScale
     )
    )) {isSetParameters=true}
   &InterpolateOverTime(
    Rate Of Change:NiagaraFloat=&FN_FrontEndFloatModifier(Gameplay=24, Front End=12)
    Target Float Value:NiagaraFloat=$Particles.AudioScale
   ) {module=InterpolateOverTime001}
   /SetVariables(Particles.AudioScale:NiagaraFloat=$Particles.InterpolateOverTime001.InterpolateOverTime_Float.Moving Average) {isSetParameters=true}
   /SetVariables(Particles.RibbonTime:NiagaraFloat=&Add_Float(
     A=$Particles.RibbonTime
     B=&Multiply_Float(
      A=$Engine.DeltaTime
      B=&Multiply_Float(
       A=&FloatFromCurve(
        FloatCurve=DataInterface() {propertyValues {Curve {keys [{arriveTangent=0.121, interpMode=RCIM_Cubic, leaveTangent=0.121, tangentMode=RCTM_User, time=0, value=1}, <at 1 3 RCIM_Cubic>]}}}
        CurveIndex=&NormalizeFloat(Value=$System.ContrailDiveAlpha, Min=0, Max=1)
        Scale Curve=&Lerp_Float(A=1, B=1.2, Alpha=$System.SmoothedAudioScale)
       )
       B=1
      )
     )
    )) {isSetParameters=true}
   /SetVariables(Particles.InterpolatedTrailVector:Vector3f=&FN_FrontEndVector3Modifier(
     Gameplay=/Niagara/DynamicInputs/Math/SlerpVector(
      Vector A:Vector3f=&Lerp_Vector(
       A=&Add_Vector(
        A=&AddVectorToPosition(
         Position=Linked() {linkedVariable {name=Emitter.TargetVector, type=Vector3f}}
         Vector=&MakeVector(
          X=&Multiply_Float(
           A=&SelectFloatFromArray(Direct Array Index=$Particles.PartitionParticles.Partition, Float Selection Array=DataInterface() {propertyValues {FloatData [1,-1]}})
           B=$Particles.AudioScale
          )
          Y=0
          Z=0
         )
        )
        B=&Multiply_Vector(
         A=&Multiply_Vector(
          A=&SelectVectorFromArray(
           Direct Array Index=$Particles.PartitionParticles.Partition
           Vector Selection Array=DataInterface() {propertyValues {FloatData [<2 1 1>, <-2 1 1>]}}
          )
          B=&VectorFromFloat(Value=&Waveform("Waveform [1]"=::NewEnumerator12, "[1] Frequency"=1, "[1] Amplitude Min/Max"=<0 1>))
         )
         B=&VectorFromCurve(
          VectorCurve=DataInterface() {propertyValues {
            XCurve {keys [
              {interpMode=RCIM_Cubic, tangentMode=RCTM_User, time=0, value=0}
              {interpMode=RCIM_Cubic, tangentMode=RCTM_User, time=0.1, value=1.75}
              {arriveTangent=-3.125, interpMode=RCIM_Cubic, leaveTangent=-3.125, time=0.2, value=0.75}
              {arriveTangent=-3.125, interpMode=RCIM_Cubic, leaveTangent=-3.125, time=0.3, value=1.125}
              {arriveTangent=-0.8333, interpMode=RCIM_Cubic, leaveTangent=-0.8333, time=0.4, value=0.125}
              <at 0.6 0.875 RCIM_Cubic>
              {arriveTangent=-1.875, interpMode=RCIM_Cubic, leaveTangent=-1.875, time=0.7, value=0.125}
              {arriveTangent=-0.4167, interpMode=RCIM_Cubic, leaveTangent=-0.4167, time=0.8, value=0.5}
              <at 1 0 RCIM_Cubic>
             ]}
            YCurve {keys [
              {interpMode=RCIM_Cubic, tangentMode=RCTM_User, time=0, value=0}
              {interpMode=RCIM_Cubic, tangentMode=RCTM_User, time=0.1, value=0.875}
              {arriveTangent=-0.8333, interpMode=RCIM_Cubic, leaveTangent=-0.8333, time=0.2, value=0.25}
              {arriveTangent=-0.4167, interpMode=RCIM_Cubic, leaveTangent=-0.4167, time=0.4, value=0.625}
              <at 0.5 0.125 RCIM_Cubic>
              {arriveTangent=-0.15, interpMode=RCIM_Cubic, leaveTangent=-0.15, time=0.7, value=0.625}
              <at 1 0.05 RCIM_Cubic>
             ]}
            ZCurve {keys [
              <at 0 0 RCIM_Cubic>
              {arriveTangent=5, interpMode=RCIM_Cubic, leaveTangent=5, time=0.1, value=0.5}
              {arriveTangent=-0.6667, interpMode=RCIM_Cubic, leaveTangent=-0.6667, time=0.2, value=1}
              {arriveTangent=-1.25, interpMode=RCIM_Cubic, leaveTangent=-1.25, time=0.4, value=0.3}
              {arriveTangent=-0.5, interpMode=RCIM_Cubic, leaveTangent=-0.5, time=0.6, value=0.5}
              <at 1 0 RCIM_Cubic>
             ]}
           }}
          CurveIndex=$Particles.PartitionParticles.Normalized Index In Partition
          Scale Curve=<0.1 0.1 2>
         )
        )
       )
       B=&SelectVectorFromArray(
        Direct Array Index=$Particles.PartitionParticles.Partition
        Vector Selection Array=DataInterface() {propertyValues {FloatData [<0.5 1 1>, <-0.5 1 1>]}}
       )
       Alpha=&NormalizeFloat(Value=$System.ContrailDiveAlpha, Min=0, Max=1)
      )
      Vector B:Vector3f=&Lerp_Vector(
       A=&SelectVectorFromArray(
        Direct Array Index=$Particles.PartitionParticles.Partition
        Vector Selection Array=DataInterface() {propertyValues {FloatData [<0.25 0.2 1>, <-0.25 0.2 1>]}}
       )
       B=&SelectVectorFromArray(
        Direct Array Index=$Particles.PartitionParticles.Partition
        Vector Selection Array=DataInterface() {propertyValues {FloatData [<0.5 1 1>, <-0.5 1 1>]}}
       )
       Alpha=&NormalizeFloat(Value=$System.ContrailDiveAlpha, Min=0, Max=1)
      )
      Alpha:NiagaraFloat=$Particles.PartitionParticles.Normalized Index In Partition
      Scale:NiagaraFloat=1
     )
     Front End=&SlerpVector(
      Vector A:Vector3f=<0 -0.35 1>
      Vector B:Vector3f=&Lerp_Vector(A=<0 -0.35 1>, B=<0 -1 1>, Alpha=&NormalizeFloat(Value=$System.ContrailDiveAlpha, Min=0, Max=1))
      Alpha:NiagaraFloat=$Particles.PartitionParticles.Normalized Index In Partition
      Scale:NiagaraFloat=1
     )
    )) {isSetParameters=true}
   &InterpolateOverTime(
    Rate Of Change:NiagaraFloat=&Lerp_Float(A=2, B=12, Alpha=&NormalizeFloat(Value=$System.ContrailDiveAlpha, Min=0, Max=1))
    Target Position Value:NiagaraPosition=$Output.SocketLocation.SampledPosition
    Variable Type=::NewEnumerator6
   )
   /SetVariables(Particles.Interpolated_WPO:Vector4f=&MakeFloat4FromVectorAndFloat(
     "Vector (RGB)"=&TransformPosition(Position=$StackContext.InterpolateOverTime.InterpolateOverTime_Position.Moving Average, Source Space=::Local, Destination Space=::World)
     "Float (Alpha)"=1
    )) {isSetParameters=true}
   /SetVariables(Particles.InterpolatedPosition:NiagaraPosition=&AddVectorToPosition(
     Position=&Lerp_Vector(
      A=$Output.SocketLocation.SampledPosition
      B=Linked() {linkedVariable {name=StackContext.InterpolateOverTime.InterpolateOverTime_Position.Moving Average, type=NiagaraPosition}}
      Alpha=$Particles.PartitionParticles.Normalized Index In Partition
     )
     Vector=&Add_Vector(
      A=&Multiply_VectorByFloat(
       Vector=$Particles.InterpolatedTrailVector
       Float=&Multiply_Float(
        A=&Multiply_Float(
         A=$Emitter.RibbonLength
         B=&FN_FrontEndFloatModifier(
          Gameplay=1
          Front End=&Lerp_Float(A=0.65, B=1.2, Alpha=&NormalizeFloat(Value=$System.ContrailDiveAlpha, Min=0, Max=1))
         )
        )
        B=$Particles.PartitionParticles.Normalized Index In Partition
       )
      )
      B=&Multiply_Vector(
       A=&Lerp_Vector(A=<1 1 1>, B=<0.2 0.2 0.2>, Alpha=&NormalizeFloat(Value=$System.ContrailDiveAlpha, Min=0, Max=1))
       B=&FN_FrontEndVector3Modifier(
        Gameplay=&Multiply_Vector(
         A=&VectorFromCurve(
          VectorCurve=DataInterface() {propertyValues {
            XCurve {keys [
              {interpMode=RCIM_Cubic, tangentMode=RCTM_User, time=0, value=0}
              {interpMode=RCIM_Cubic, tangentMode=RCTM_User, time=0.1, value=0.75}
              {arriveTangent=5.6827, interpMode=RCIM_Cubic, leaveTangent=5.6827, time=0.2, value=0.75}
              {arriveTangent=1.6667, interpMode=RCIM_Cubic, leaveTangent=1.6667, time=0.32, value=2}
              {arriveTangent=-0.8928, interpMode=RCIM_Cubic, leaveTangent=-0.8928, time=0.5, value=1.25}
              {arriveTangent=-1.875, interpMode=RCIM_Cubic, leaveTangent=-1.875, time=0.6, value=1.75}
              {arriveTangent=1.25, interpMode=RCIM_Cubic, leaveTangent=1.25, time=0.7, value=0.875}
              {arriveTangent=-2.9167, interpMode=RCIM_Cubic, leaveTangent=-2.9167, time=0.8, value=2}
              <at 1 0 RCIM_Cubic>
             ]}
            YCurve {keys [
              {interpMode=RCIM_Cubic, tangentMode=RCTM_User, time=0, value=0}
              {interpMode=RCIM_Cubic, tangentMode=RCTM_User, time=0.1, value=0.875}
              {arriveTangent=0.4167, interpMode=RCIM_Cubic, leaveTangent=0.4167, time=0.2, value=0.5}
              {arriveTangent=0.4167, interpMode=RCIM_Cubic, leaveTangent=0.4167, time=0.4, value=1}
              {arriveTangent=-0.4167, interpMode=RCIM_Cubic, leaveTangent=-0.4167, time=0.5, value=0.625}
              {arriveTangent=-1.15, interpMode=RCIM_Cubic, leaveTangent=-1.15, time=0.7, value=0.875}
              <at 1 0.05 RCIM_Cubic>
             ]}
            ZCurve {keys [
              <at 0 0 RCIM_Cubic>
              {arriveTangent=5, interpMode=RCIM_Cubic, leaveTangent=5, time=0.1, value=0.5}
              {arriveTangent=-0.6667, interpMode=RCIM_Cubic, leaveTangent=-0.6667, time=0.2, value=1}
              {arriveTangent=-0.3125, interpMode=RCIM_Cubic, leaveTangent=-0.3125, time=0.4, value=0.3}
              {arriveTangent=-0.5, interpMode=RCIM_Cubic, leaveTangent=-0.5, time=0.6, value=0.875}
              <at 1 0 RCIM_Cubic>
             ]}
           }}
          CurveIndex=$Particles.PartitionParticles.Normalized Index In Partition
          Scale Curve=&SelectVectorFromArray(
           Direct Array Index=$Particles.PartitionParticles.Partition
           Vector Selection Array=DataInterface() {propertyValues {FloatData [<20 7.5 5>, <-20 7.5 5>]}}
          )
         )
         B=&Lerp_Vector(A=<0.75 0.75 1>, B=<0.25 0.1 1>, Alpha=$Particles.PartitionParticles.Normalized Index In Partition)
        )
        Front End=&Multiply_Vector(
         A=&VectorFromCurve(
          VectorCurve=DataInterface() {propertyValues {
            XCurve {keys [
              {interpMode=RCIM_Cubic, tangentMode=RCTM_User, time=0, value=0}
              {arriveTangent=3.8942, interpMode=RCIM_Cubic, leaveTangent=3.8942, tangentMode=RCTM_User, time=0.1, value=0.75}
              {arriveTangent=5.6827, interpMode=RCIM_Cubic, leaveTangent=5.6827, time=0.2, value=1.25}
              {arriveTangent=2.5, interpMode=RCIM_Cubic, leaveTangent=2.5, time=0.32, value=2}
              <at 0.5 2 RCIM_Cubic>
              {arriveTangent=0.625, interpMode=RCIM_Cubic, leaveTangent=0.625, time=0.6, value=2}
              <at 0.7 2.125 RCIM_Cubic>
              {arriveTangent=-2.5, interpMode=RCIM_Cubic, leaveTangent=-2.5, time=0.8, value=2}
              <at 1 1.375 RCIM_Cubic>
             ]}
            YCurve {keys [
              {interpMode=RCIM_Cubic, tangentMode=RCTM_User, time=0, value=0}
              {interpMode=RCIM_Cubic, tangentMode=RCTM_User, time=0.1, value=0.875}
              {arriveTangent=1.25, interpMode=RCIM_Cubic, leaveTangent=1.25, time=0.3, value=1.125}
              {arriveTangent=0.625, interpMode=RCIM_Cubic, leaveTangent=0.625, time=0.4, value=1.25}
              {arriveTangent=-0.8333, interpMode=RCIM_Cubic, leaveTangent=-0.8333, time=0.5, value=1.25}
              {arriveTangent=-1, interpMode=RCIM_Cubic, leaveTangent=-1, time=0.7, value=1}
              <at 1 0.75 RCIM_Cubic>
             ]}
            ZCurve {keys [
              <at 0 0 RCIM_Cubic>
              {arriveTangent=5, interpMode=RCIM_Cubic, leaveTangent=5, time=0.1, value=0.5}
              {arriveTangent=2.5, interpMode=RCIM_Cubic, leaveTangent=2.5, time=0.2, value=1}
              {arriveTangent=0.625, interpMode=RCIM_Cubic, leaveTangent=0.625, time=0.4, value=1.25}
              {arriveTangent=-0.8333, interpMode=RCIM_Cubic, leaveTangent=-0.8333, time=0.6, value=1.25}
              <at 1 0.75 RCIM_Cubic>
             ]}
           }}
          CurveIndex=$Particles.PartitionParticles.Normalized Index In Partition
          Scale Curve=&SelectVectorFromArray(
           Direct Array Index=$Particles.PartitionParticles.Partition
           Vector Selection Array=DataInterface() {propertyValues {FloatData [<20 7.5 10>, <-20 7.5 10>]}}
          )
         )
         B=&Lerp_Vector(A=<0.75 0.75 1>, B=<0.25 0.1 1>, Alpha=$Particles.PartitionParticles.Normalized Index In Partition)
        )
       )
      )
     )
    )) {isSetParameters=true}
   /SetVariables(Particles.RibbonFacing=&Multiply_Vector(
     A=&SelectVectorFromArray(
      Direct Array Index=$Particles.PartitionParticles.Partition
      Vector Selection Array=DataInterface() {propertyValues {FloatData [<1 -1 0>, <1 1 0>]}}
     )
     B=<1 1 1>
    )) {isSetParameters=true}
   /SetVariables(Particles.RibbonWidth=&Multiply_Float(
     A=&Multiply_Float(
      A=&Lerp_Float(
       A=&Lerp_Float(A=25, B=5, Alpha=&NormalizeFloat(Value=$System.ContrailDiveAlpha, Min=0, Max=1))
       B=&Lerp_Float(A=175, B=120, Alpha=&NormalizeFloat(Value=$System.ContrailDiveAlpha, Min=0, Max=1))
       Alpha=&FloatFromCurve(
        FloatCurve=DataInterface() {propertyValues {Curve {keys [
            {arriveTangent=5.3996, interpMode=RCIM_Cubic, leaveTangent=5.3996, tangentMode=RCTM_User, time=0, value=0}
            {arriveTangent=2.7644, interpMode=RCIM_Cubic, leaveTangent=2.7644, tangentMode=RCTM_User, time=0.0788, value=0.3735}
            {arriveTangent=1.0222, interpMode=RCIM_Cubic, leaveTangent=1.0222, tangentMode=RCTM_User, time=0.2042, value=0.6228}
            {arriveTangent=-0.5762, interpMode=RCIM_Cubic, leaveTangent=-0.5762, tangentMode=RCTM_User, time=0.5, value=0.9}
            {arriveTangent=-3.049, interpMode=RCIM_Cubic, leaveTangent=-3.049, tangentMode=RCTM_User, time=1, value=0.2}
           ]}}}
        CurveIndex=$Particles.PartitionParticles.Normalized Index In Partition
       )
      )
      B=&Lerp_Float(A=0.8, B=1.2, Alpha=$Particles.AudioScale)
     )
     B=&FN_FrontEndFloatModifier(
      Gameplay=1
      Front End=&Lerp_Float(A=0.4, B=1.1, Alpha=&NormalizeFloat(Value=$System.ContrailDiveAlpha, Min=0, Max=1))
     )
    )) {isSetParameters=true}
   &DynamicMaterialParameters(
    Index 0 Param 1=$Particles.RibbonTime
    Index 2 Param 1=$System.SmoothedAudioScale
    Module.Index0Param0Write=true
    Module.Index0Param1Write=false
    Module.Index0Param2Write=false
    Module.Index0Param3Write=false
    Module.Index2Param0Write=true
    Module.Index2Param1Write=false
    Module.Index2Param2Write=false
    Module.Index2Param3Write=false
    Write Parameter Index 2=true
   ) {module=DynamicMaterialParameters001}
  ]
  rendererClasses [&NiagaraRibbonRendererProperties]
  renderers [
   {rendererClass=&NiagaraRibbonRendererProperties}
   {rendererClass=&NiagaraRibbonRendererProperties, rendererIndex=1}
  ]
 }
 Smoke {
  emitterUpdateScript [
   &EmitterState()
   &SpawnRate(SpawnRate=15, Use Spawn Probability=false)
   &SpawnRate(SpawnGroup=1, SpawnRate=10, Use Spawn Probability=false) {module=SpawnRate001}
  ]
  particleSpawnScript [
   &InitializeParticle(
    Color=&RandomRangeLinearColor(
     Minimum=&RandomRangeLinearColor(Minimum=<rgba 0.6604 0 1 0.7>, Maximum=<rgba 0.4167 0 1 0.7>)
     Maximum=&RandomRangeLinearColor(Minimum=<rgba 0 1 0.7667 0.7>, Maximum=<rgba 0 0.5833 1 0.7>)
    )
    Lifetime=&UniformRangedFloat(Minimum=0.5, Maximum=0.75)
    Sprite Rotation Mode=::NewEnumerator2
    Sprite Size Mode=::NewEnumerator3
    Uniform Sprite Size=&Multiply_Float(
     A=&RandomRangeFloat(Minimum=40, Maximum=60)
     B=&MaskFloatBySpawnGroup(Masked Float=1, Passthrough Float=&RandomRangeFloat(Minimum=0.4, Maximum=0.6))
    )
    UsePositionOffset=false
    Write Material Random=true
   )
   &PartitionParticles(Number Of Partitions=4, Write Partition to Ribbon ID=::NewEnumerator0)
   &DynamicMaterialParameters(
    Index 0 Param 1=&RandomRangeFloat(Minimum=-0.05, Maximum=0.05)
    Index 0 Param 2=&RandomRangeFloat(Minimum=0, Maximum=32)
    Module.Index0Param0Write=true
    Module.Index0Param1Write=true
    Module.Index0Param2Write=false
    Module.Index0Param3Write=false
   )
   &AddVelocity(
    Velocity=&FN_FrontEndVector3Modifier(
     Gameplay=&Lerp_Vector(
      A=&FN_FrontEndVector3Modifier(Gameplay=&MaskVectorBySpawnGroup(Masked Vector=<0 75 100>, Passthrough Vector=<0 50 100>), Front End=<0 0 100>)
      B=&MakeVector(
       X=&SelectFloatFromArray(
        Direct Array Index=$Particles.PartitionParticles.Partition
        Float Selection Array=DataInterface() {propertyValues {FloatData [50,-50,25,-25]}}
       )
       Y=75
       Z=15
      )
      Alpha=$System.Dive
     )
     Front End=&Lerp_Vector(A=<0 -25 100>, B=<0 -150 150>, Alpha=$System.Dive)
    )
    Velocity Speed Scale=&Multiply_Float(
     A=&MaskFloatBySpawnGroup(Masked Float=1, Passthrough Float=&RandomRangeFloat(Minimum=1, Maximum=1.5))
     B=&FN_FrontEndFloatModifier(Gameplay=&MaskFloatBySpawnGroup(Masked Float=0.5, Passthrough Float=0.2), Front End=1)
    )
   )
  ]
  particleUpdateScript [
   &ParticleState()
   &SocketLocation(
    SamplePosOffset=true
    Sampled Position Offset=&FN_FrontEndVector3Modifier(
     Gameplay=<0 0 0>
     Front End=&Lerp_Vector(
      A=&SelectVectorFromArray(
       Direct Array Index=$Particles.PartitionParticles.Partition
       Vector Selection Array=DataInterface() {propertyValues {FloatData [<-10 0 0>, <10 0 0>, <10 0 0>, <-10 0 0>]}}
      )
      B=&SelectVectorFromArray(
       Direct Array Index=$Particles.PartitionParticles.Partition
       Vector Selection Array=DataInterface() {propertyValues {FloatData [<-10 0 0>, <10 0 0>, <15 0 0>, <-15 0 0>]}}
      )
      Alpha=$System.Dive
     )
    )
    Socket Index=$Particles.PartitionParticles.Partition
    Socket Reader=$System.Socket Reader
    Socket Sampling Mode=::NewEnumerator3
    VelocitySpeedLimit=false
   )
   &SolveForcesAndVelocity(Clamp Velocity=false, Limit Acceleration=false)
   &ScaleColor(
    Scale Alpha=&FloatFromCurve(
     FloatCurve=DataInterface() {propertyValues {Curve {keys [
         {arriveTangent=-0.4774, interpMode=RCIM_Cubic, leaveTangent=-0.4774, tangentMode=RCTM_User, time=0, value=0}
         <at 0.5023 0.5898 RCIM_Cubic>
         <at 1.0011 0 RCIM_Cubic>
        ]}}}
     Scale Curve=&FN_FrontEndFloatModifier(Gameplay=&MaskFloatBySpawnGroup(Masked Float=0.5, Passthrough Float=1.5), Front End=1)
    )
    Scale RGB=&Multiply_Vector(
     A=&MakeVectorFromLinearColorRGB(Color=&FN_Select_Color_from_Time_Of_Day(
       Use Manual Time of Day Thresholds=true
       Morning=<rgba 0.942 0.4238 2.5 1>
       Day=<rgba 1 1 1 1>
       Evening=<rgba 1.1427 0.821 2 1>
       Night=<rgba 0 0.3541 1 1>
       Evening Threshold=22
      ))
     B=&VectorFromCurve(
      VectorCurve=DataInterface() {propertyValues {XCurve {keys [<at 0 1>,<at 1 0.9>]}, YCurve {keys [<at 0 1>,<at 1 0.9>]}, ZCurve {keys [<at 0 1>,<at 1 2>]}}}
      Scale Curve=&Multiply_VectorByFloat(
       Vector=&Multiply_VectorByFloat(Vector=<15 15 15>, Float=&MaskFloatBySpawnGroup(Masked Float=1, Passthrough Float=2))
       Float=&Lerp_Float(A=1, B=2, Alpha=$System.SmoothedAudioScale)
      )
     )
    )
    ScaleA=true
    ScaleRGB=true
   )
   &ScaleSpriteSize(
    Uniform Curve Scale=&Multiply_Float(
     A=&SelectFloatFromArray(
      Direct Array Index=$Particles.PartitionParticles.Partition
      Float Selection Array=DataInterface() {propertyValues {FloatData [1.75,1.75,0.8,0.8]}}
     )
     B=&Multiply_Float(
      A=&Lerp_Float(A=1, B=1.25, Alpha=$System.SmoothedAudioScale)
      B=&FN_FrontEndFloatModifier(Gameplay=&MaskFloatBySpawnGroup(Masked Float=0.8, Passthrough Float=0.8), Front End=0.85)
     )
    )
    Uniform Curve Sprite Scale=DataInterface() {propertyValues {Curve {keys [
        {arriveTangent=2.3771, interpMode=RCIM_Cubic, leaveTangent=2.3771, tangentMode=RCTM_User, time=0, value=0}
        {interpMode=RCIM_Cubic, leaveTangent=-0.4663, time=1, value=1}
       ]}}}
   )
   &CameraOffset(Camera Offset Amount=50)
   &InterpolateOverTime(
    Rate Of Change:NiagaraFloat=&MaskFloatBySpawnGroup(
     Masked Float=&RandomRangeFloat(Minimum=2, Maximum=3)
     Passthrough Float=&RandomRangeFloat(Minimum=12, Maximum=24)
    )
    Target Position Value:NiagaraPosition=$Particles.Position
    Variable Type=::NewEnumerator6
   )
   /SetVariables(Particles.Position=$Particles.InterpolateOverTime.InterpolateOverTime_Position.Moving Average) {isSetParameters=true}
   &KillParticlesInVolume(Invert Volume=true, Kill Volume Enabled=false, Sphere Radius=25, Volume Origin:Vector3f=$Output.SocketLocation.SampledPosition)
   &Timeline(
    Play=&SetBoolByFloatComparison(
     A=&NormalizeDistanceRange(
      Start Position:NiagaraPosition=$Output.SocketLocation.SampledPosition
      End Position:NiagaraPosition=$Particles.Position
      Distance:NiagaraFloat=25
      Invert Normalized Range:NiagaraBool=false
     )
     B=0.25
     Comparison Type=::NewEnumerator2
    )
    Play Rate=2.5
    Rewind Rate=0.2
   )
   &ScaleColor(Scale Alpha=$Particles.Timeline.InverseNormalizedTime, ScaleA=true, ScaleRGB=false) {module=ScaleColor001}
  ]
  rendererClasses [&NiagaraSpriteRendererProperties]
  renderers [{rendererClass=&NiagaraSpriteRendererProperties}]
 }
 Specs {
  emitterUpdateScript [
   &EmitterState()
   &SpawnRate(SpawnRate=6, Use Spawn Probability=false)
   &SpawnPerFrame(
    Spawn:NiagaraBool=&SetBoolByFloatComparison(A=$System.AudioScale, B=0.75)
    Spawn Count:NiagaraInt32=&RandomRangeInt(Minimum=12, Maximum=24)
    Use Spawn Probability=false
   )
  ]
  particleSpawnScript [
   &InitializeParticle(
    Color=<rgba 0 10.4137 20 1>
    Lifetime=&RandomRangeFloat(Minimum=0.5, Maximum=1)
    Sprite Rotation Mode=::NewEnumerator2
    Sprite Size Mode=::NewEnumerator3
    Sprite UV Mode=::NewEnumerator3
    Uniform Sprite Size=&RandomRangeFloat(Minimum=1, Maximum=3)
    UsePositionOffset=false
    Write Material Random=true
   )
   &PartitionParticles(Number Of Partitions=4, Write Partition to Ribbon ID=::NewEnumerator0)
  ]
  particleUpdateScript [
   &ParticleState()
   &SocketLocation(
    SamplePosOffset=true
    Sampled Position Offset=&AddVectorToPosition(
     Position=&Multiply_VectorByFloat(Vector=&RandomRangeVector(Minimum=<-1 -1 -1>, Maximum=<1 1 1>), Float=10)
     Vector=&Lerp_Vector(A=<0 0 -10>, B=<0 0 20>, Alpha=$Particles.NormalizedAge)
    )
    Sampled Position Offset Coordinate Space:ENiagaraSkelMeshTransforms=::NewEnumerator2
    Socket Index=$Particles.PartitionParticles.Partition
    Socket Reader=$System.Socket Reader
    Socket Sampling Mode=::NewEnumerator3
    VelocitySpeedLimit=false
   )
   &CameraOffset(Camera Offset Amount=50)
   &ScaleColor(
    Scale Alpha=&FloatFromCurve(FloatCurve=DataInterface() {propertyValues {Curve {keys [
         {interpMode=RCIM_Cubic, tangentMode=RCTM_User, time=0, value=0}
         {interpMode=RCIM_Cubic, tangentMode=RCTM_User, time=0.5, value=1}
         {tangentMode=RCTM_User, time=1, value=0}
        ]}}})
    Scale RGB=&VectorFromFloat(Value=&Lerp_Float(
      A=0.5
      B=&RandomRangeFloat(Minimum=2.5, Maximum=5)
      Alpha=&NormalizeFloat(Value=$System.SmoothedAudioScale, Min=0, Max=1)
     ))
    ScaleA=true
    ScaleRGB=true
   )
   &ScaleSpriteSize(Scale Factor=&Vector2DFromFloat(Value=&Multiply_Float(
      A=&FloatFromCurve(
       FloatCurve=DataInterface() {propertyValues {Curve {keys [
           {interpMode=RCIM_Cubic, tangentMode=RCTM_User, time=0, value=1}
           {interpMode=RCIM_Cubic, tangentMode=RCTM_User, time=0.5, value=1.1}
           {tangentMode=RCTM_User, time=1, value=1}
          ]}}}
       Scale Curve=&SelectFloatFromArray(
        Direct Array Index=$Particles.PartitionParticles.Partition
        Float Selection Array=DataInterface() {propertyValues {FloatData [0.8,0.8,0.4,0.4]}}
       )
      )
      B=&Lerp_Float(A=0.5, B=5, Alpha=&NormalizeFloat(Value=$System.SmoothedAudioScale, Min=0, Max=1))
     )))
   &FN_Sprite_Attributes(
    Color=&MultiplyLinearColor(
     Linear Color=<rgba 0 0.581 1 0.5>
     Scale Factor=&MakeFloat4FromVectorAndFloat(
      "Vector (RGB)"=&VectorFromFloat(Value=&Lerp_Float(A=0.5, B=2.5, Alpha=$System.SmoothedAudioScale))
      "Float (Alpha)"=&FloatFromCurve(FloatCurve=DataInterface() {propertyValues {Curve {keys [
           {interpMode=RCIM_Cubic, tangentMode=RCTM_User, time=0, value=0}
           {interpMode=RCIM_Cubic, tangentMode=RCTM_User, time=0.5, value=1}
           {tangentMode=RCTM_User, time=1, value=0}
          ]}}})
     )
    )
    Create Color Attribute=true
    Create Position Attribute=true
    Create SpriteSize Attribute=true
    Position:Vector3f=$Output.SocketLocation.SampledPosition
    SpriteSize=&Multiply_Vector2D(A=$Particles.SpriteSize, B=<5 5>)
    Write to Module Namespace=true
   )
   &DynamicMaterialParameters(
    Index 0 Param 1=1.5
    Index 0 Param 3=5
    Index 0 Param 4=&RandomRangeFloat(Minimum=0, Maximum=1)
    Index 2 Param 1=3
    Index 2 Param 2=1
    Index 2 Param 3=1
    Index 2 Param 4=1
    Module.Param0WriteEnabled=true
    Module.Param0WriteEnabled2=true
    Module.Param1WriteEnabled=true
    Module.Param1WriteEnabled2=true
    Module.Param2WriteEnabled=true
    Module.Param2WriteEnabled2=true
    Module.Param3WriteEnabled=true
    Module.Param3WriteEnabled2=true
    Write Parameter Index 2=true
   )
  ]
  rendererClasses [&NiagaraSpriteRendererProperties]
  renderers [
   {rendererClass=&NiagaraSpriteRendererProperties}
   {rendererClass=&NiagaraSpriteRendererProperties, rendererIndex=1}
  ]
 }
}
summary {
 systemName=NS_Contrail_KelpLinen
 emitters [
  {emitterName=Butterfly, rendererClasses [&NiagaraMeshRendererProperties, &NiagaraSpriteRendererProperties]}
  {emitterName=Ribbons, rendererClasses [&NiagaraRibbonRendererProperties]}
  {emitterName=Smoke, rendererClasses [&NiagaraSpriteRendererProperties]}
  {emitterName=Decal_Glow, rendererClasses [&NiagaraSpriteRendererProperties]}
  {emitterName=Hands_Support_Shapes, rendererClasses [&NiagaraMeshRendererProperties]}
  {emitterName=Specs, rendererClasses [&NiagaraSpriteRendererProperties]}
  {emitterName=BlackBits, rendererClasses [&NiagaraSpriteRendererProperties]}
  {emitterName=PixieDust, rendererClasses [&NiagaraSpriteRendererProperties]}
  {bEnabled=false, emitterName=FN_Debug, rendererClasses [&NiagaraSpriteRendererProperties]}
 ]
 userVariables [
  User.bIsFrontEnd {type=NiagaraBool}
  User.bIsFrontEndPreview {type=NiagaraBool}
  User.ParaGlideLeanAlpha {type=NiagaraFloat, defaultValue=0, description=""}
  User.RibbonIDOffset {type=NiagaraInt32}
 ]
}
